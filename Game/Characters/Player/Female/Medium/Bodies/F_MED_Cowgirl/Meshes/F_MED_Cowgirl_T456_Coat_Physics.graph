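# Physics asset for the Cowgirl T456 coat on the medium female body, giving the coat's skirt bones jointed collision bodies.
# Front, diagonal, outer, back and center chains hang from the pelvis with angular limits, using capsules at the front and boxes elsewhere; the center splits left and right.
# Cross-constraints tie neighboring chains, mostly near the hem.
# Pelvis and leg capsules plus a wide, flat root box at ground level supply body and floor collision shapes.
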PhysicsAsset /Game/Characters/Player/Female/Medium/Bodies/F_MED_Cowgirl/Meshes/F_MED_Cowgirl_T456_Coat_Physics
bodies {
 calf_l [{shapeType=Capsule, length=47.972, radius=9.25, center=<-23.2226 -4.6248 0>, rotation=<rot 90 -0.5968 -5.5968>}]
 calf_r [{shapeType=Capsule, length=47.972, radius=9.25, center=<23.2226 4.6248 0>, rotation=<rot 90 -0.5968 -5.5968>}]
 dyn_skirt_bk_01_l [{shapeType=Box, extentX=2, extentY=16, extentZ=18, center=<0.0004 0 -9.4001>}]
 dyn_skirt_bk_01_r [{shapeType=Box, extentX=2, extentY=16, extentZ=18, center=<0 0 9.4001>}]
 dyn_skirt_bk_02_l [{shapeType=Box, extentX=2, extentY=16, extentZ=22, center=<0.0004 0 -11.8494>}]
 dyn_skirt_bk_02_r [{shapeType=Box, extentX=2, extentY=16, extentZ=22, center=<0 0 11.8494>}]
 dyn_skirt_bk_03_l [{shapeType=Box, extentX=2, extentY=16, extentZ=16, center=<0.0006 0 -7.9924>}]
 dyn_skirt_bk_03_r [{shapeType=Box, extentX=4.7723, extentY=16, extentZ=16, center=<0 0 7.9924>}]
 dyn_skirt_bk_04_l [
  {shapeType=Box, extentX=2, extentY=11.8151, extentZ=17.6194, center=<-0.6007 5.5683 -7.7947>, rotation=<rot 0 5 0>}
  {shapeType=Box, extentX=2, extentY=11.8151, extentZ=15.5043, center=<-0.7976 -5.6467 -7.0733>, rotation=<rot 0 -10.0016 0>}
 ]
 dyn_skirt_bk_04_r [
  {shapeType=Box, extentX=4.7723, extentY=11.8151, extentZ=17.6194, center=<0.6007 -5.5683 7.7947>, rotation=<rot 0 5.0002 0>}
  {shapeType=Box, extentX=4.7723, extentY=11.8151, extentZ=15.5043, center=<0.7976 5.6467 7.0733>, rotation=<rot 0 -10.0016 0>}
 ]
 dyn_skirt_fr_01_l [{shapeType=Capsule, length=16, radius=1.25, shapeName=dyn_skirt_fr_01_l_capsule}]
 dyn_skirt_fr_01_r [{shapeType=Capsule, length=16, radius=1.25, shapeName=dyn_skirt_fr_01_r_capsule}]
 dyn_skirt_fr_02_l [{shapeType=Capsule, length=16, radius=1.25, shapeName=dyn_skirt_fr_02_l_capsule}]
 dyn_skirt_fr_02_r [{shapeType=Capsule, length=16, radius=1.25, shapeName=dyn_skirt_fr_02_r_capsule}]
 dyn_skirt_fr_02a_l [{shapeType=Capsule, length=15.25, radius=2, shapeName=dyn_skirt_fr_02a_l_capsule, center=<-0.7316 0.0001 0>}]
 dyn_skirt_fr_02a_r [{shapeType=Capsule, length=15, radius=2.25, shapeName=dyn_skirt_fr_02a_r_capsule, center=<0.6301 0 0>}]
 dyn_skirt_fr_03_l [{shapeType=Capsule, length=24.25, radius=2.25, shapeName=dyn_skirt_fr_03_l_capsule, center=<-0.5762 -0.0001 -10.7671>}]
 dyn_skirt_fr_03_r [{shapeType=Capsule, length=20.5, radius=3, shapeName=dyn_skirt_fr_03_r_capsule, center=<0.835 -0.0001 10.85>}]
 dyn_skirt_fr_04_l [{shapeType=Capsule, length=15, radius=2.25, shapeName=dyn_skirt_fr_04_l_capsule, center=<-0.5761 -0.0001 -9.1346>}]
 dyn_skirt_fr_04_r [{shapeType=Capsule, length=14.25, radius=3, shapeName=dyn_skirt_fr_04_r_capsule, center=<0.8353 -0.0001 6.8331>}]
 dyn_skirt_fr_05_l [{shapeType=Capsule, length=15, radius=2.25, shapeName=dyn_skirt_fr_05_l_capsule, center=<-0.5762 0 -8.6646>}]
 dyn_skirt_fr_05_r [{shapeType=Capsule, length=14.25, radius=3, shapeName=dyn_skirt_fr_05_r_capsule, center=<0.8353 0 7.785>}]
 dyn_skirt_frdiag_01_l [{shapeType=Capsule, length=18.5, radius=1.25, shapeName=dyn_skirt_frdiag_01_l_capsule, center=<-0.397 0.0002 -9.8505>}]
 dyn_skirt_frdiag_01_r [{shapeType=Capsule, length=17.25, radius=1.25, shapeName=dyn_skirt_frdiag_01_r_capsule, center=<0.6615 0.0001 10.0682>}]
 dyn_skirt_frdiag_02_l [{shapeType=Capsule, length=23.75, radius=1.75, shapeName=dyn_skirt_frdiag_02_l_capsule, center=<-0.8306 0.0001 -7.9727>}]
 dyn_skirt_frdiag_02_r [{shapeType=Capsule, length=27, radius=1.75, shapeName=dyn_skirt_frdiag_02_r_capsule, center=<0.6616 0.0005 10.8132>}]
 dyn_skirt_frdiag_03_l [{shapeType=Capsule, length=17.25, radius=1.75, shapeName=dyn_skirt_frdiag_03_l_capsule, center=<-0.397 0 -9.0384>}]
 dyn_skirt_frdiag_03_r [{shapeType=Capsule, length=17.75, radius=2.5, shapeName=dyn_skirt_frdiag_03_r_capsule, center=<0.6616 0.0001 9.6762>}]
 dyn_skirt_frdiag_04_l [{shapeType=Capsule, length=17.25, radius=1.75, shapeName=dyn_skirt_frdiag_04_l_capsule, center=<-0.397 0 -9.0384>}]
 dyn_skirt_frdiag_04_r [{shapeType=Capsule, length=17.75, radius=2.5, shapeName=dyn_skirt_frdiag_04_r_capsule, center=<0.6616 0.0001 9.6762>}]
 dyn_skirt_mid_01 [{shapeType=Box, extentX=2, extentY=15.5, extentZ=14.5, center=<-0.0002 0 -6.956>}]
 dyn_skirt_mid_02 [{shapeType=Box, extentX=2, extentY=15.5, extentZ=15, center=<0.0003 0 -7.4813>}]
 dyn_skirt_mid_03_l [{shapeType=Box, extentX=2, extentY=9.5, extentZ=14, center=<0.0006 0 -7.3312>}]
 dyn_skirt_mid_03_r [{shapeType=Box, extentX=2, extentY=9.5, extentZ=14, center=<0 0 7.3312>}]
 dyn_skirt_mid_04_l [{shapeType=Box, extentX=2, extentY=9.5, extentZ=13, center=<0.0004 0 -6.7827>}]
 dyn_skirt_mid_04_r [{shapeType=Box, extentX=3.2723, extentY=9.5, extentZ=13, center=<0 0 6.7827>}]
 dyn_skirt_mid_05_l [{shapeType=Box, extentX=2, extentY=14.6503, extentZ=16.8957, center=<0 0 -7.675>, rotation=<rot 0 14.7411 0>}]
 dyn_skirt_mid_05_r [{shapeType=Box, extentX=3.7723, extentY=14.6503, extentZ=16.8957, center=<0 0 7.4724>, rotation=<rot 0 14.7418 0>}]
 dyn_skirt_out_01_l [{shapeType=Box, extentX=2, extentY=16, extentZ=20.5, center=<0.0002 0 -10.7113>}]
 dyn_skirt_out_01_r [{shapeType=Box, extentX=2, extentY=16, extentZ=20.5, center=<0 0 10.7113>}]
 dyn_skirt_out_02_l [{shapeType=Box, extentX=2, extentY=16, extentZ=21, center=<0.0005 0 -11.3306>, rotation=<rot 0 -10.0005 0>}]
 dyn_skirt_out_02_r [{shapeType=Box, extentX=2, extentY=16, extentZ=21, center=<0 0 11.3306>, rotation=<rot 0 -10.0003 0>}]
 dyn_skirt_out_03_l [{shapeType=Box, extentX=4.0444, extentY=16, extentZ=17.5, center=<0.0005 0 -8.8006>, rotation=<rot 0 -10.0005 0>}]
 dyn_skirt_out_03_r [{shapeType=Box, extentX=2, extentY=16, extentZ=17.5, center=<0 0 8.8006>, rotation=<rot 0 -10.0003 0>}]
 dyn_skirt_out_04_l [
  {shapeType=Box, extentX=4.0444, extentY=11.8151, extentZ=20.6558, center=<-0.6007 5.5683 -9.9586>, rotation=<rot 0 9.9999 0>}
  {shapeType=Box, extentX=4.0444, extentY=11.8151, extentZ=16.9504, center=<-1.6433 -5.5987 -7.8916>, rotation=<rot 0 -20.0004 0>}
 ]
 dyn_skirt_out_04_r [
  {shapeType=Box, extentX=2, extentY=9.3151, extentZ=20.6558, center=<0.6007 -5.5683 9.9586>, rotation=<rot 0 9.9999 0>}
  {shapeType=Box, extentX=2, extentY=11.8151, extentZ=16.9504, center=<1.6433 5.5987 7.8916>, rotation=<rot 0 -20.0004 0>}
 ]
 foot_l [{shapeType=Capsule, length=17.3588, radius=8.0599, center=<-4.0106 5.9761 0>, rotation=<rot 0 0 90>}]
 foot_r [{shapeType=Capsule, length=17.3588, radius=8.0599, center=<4.0106 -5.9761 0>, rotation=<rot 0 0 90>}]
 pelvis [
  {shapeType=Capsule, length=3.4375, radius=11.8125, center=<0 0.2363 0.0001>, rotation=<rot 0 -90 0>}
  {shapeType=Capsule, length=4.5, radius=10.75, center=<0.0365 -0.7316 0.0001>, rotation=<rot 0 -90 0>}
 ]
 root [{shapeType=Box, extentX=154.0568, extentY=156.9491, extentZ=7.5, center=<0 0 -3.6662>}]
 thigh_l [
  {shapeType=Capsule, length=10.8657, radius=8.483, center=<-15.3436 0.8967 -0.7739>, rotation=<rot 53.5541 -85.1927 -80.6738>}
  {shapeType=Capsule, length=18.5783, radius=8.7937, center=<-25.5485 0.1351 -0.8429>, rotation=<rot 54.1694 -87.401 -92.0629>}
  {shapeType=Capsule, length=10.8657, radius=8.483, center=<-4.8779 -0.2385 -0.0907>, rotation=<rot 78.0683 -162.7812 -149.1106>}
 ]
 thigh_r [{shapeType=Capsule, length=27, radius=9.25, center=<21.7203 -1.4449 -0.7045>, rotation=<rot 84.9748 -180 -180>}]
}
constraints [
 {bone1Name=dyn_skirt_fr_03_r, bone2Name=dyn_skirt_fr_02a_r, swing1LimitDegrees=90, swing2LimitDegrees=90, twistLimitDegrees=30}
 {bone1Name=dyn_skirt_fr_03_l, bone2Name=dyn_skirt_fr_02a_l, swing1LimitDegrees=50, swing2LimitDegrees=135, twistLimitDegrees=10}
 {bone1Name=thigh_l, bone2Name=pelvis, swing1LimitDegrees=45, swing2LimitDegrees=45, twistLimitDegrees=45}
 {bone1Name=calf_l, bone2Name=thigh_l, swing1LimitDegrees=45, swing2LimitDegrees=45, twistLimitDegrees=45}
 {bone1Name=thigh_r, bone2Name=pelvis, swing1LimitDegrees=45, swing2LimitDegrees=45, twistLimitDegrees=45}
 {bone1Name=calf_r, bone2Name=thigh_r, swing1LimitDegrees=45, swing2LimitDegrees=45, twistLimitDegrees=45}
 {bone1Name=dyn_skirt_fr_01_r, bone2Name=pelvis, swing1LimitDegrees=45, swing2LimitDegrees=60, twistLimitDegrees=2}
 {bone1Name=dyn_skirt_fr_02_r, bone2Name=dyn_skirt_fr_01_r, swing1LimitDegrees=90, swing2LimitDegrees=60, twistLimitDegrees=10}
 {bone1Name=dyn_skirt_fr_05_r, bone2Name=dyn_skirt_fr_04_r, swing1LimitDegrees=80, swing2LimitDegrees=80, twistLimitDegrees=10}
 {bone1Name=dyn_skirt_fr_01_l, bone2Name=pelvis, swing1LimitDegrees=20, swing2LimitDegrees=80, twistLimitDegrees=10}
 {bone1Name=dyn_skirt_fr_02_l, bone2Name=dyn_skirt_fr_01_l, swing1LimitDegrees=20, swing2LimitDegrees=20, twistLimitDegrees=10}
 {bone1Name=dyn_skirt_fr_04_l, bone2Name=dyn_skirt_fr_03_l, swing1LimitDegrees=80, swing2LimitDegrees=100, twistLimitDegrees=30}
 {bone1Name=dyn_skirt_fr_05_l, bone2Name=dyn_skirt_fr_04_l, swing1LimitDegrees=135, swing2LimitDegrees=135, twistLimitDegrees=10}
 {bone1Name=dyn_skirt_fr_04_r, bone2Name=dyn_skirt_fr_03_r, swing1LimitDegrees=80, swing2LimitDegrees=100, twistLimitDegrees=30}
 {bone1Name=dyn_skirt_mid_01, bone2Name=pelvis, swing1LimitDegrees=10, swing2LimitDegrees=60, twistLimitDegrees=2}
 {bone1Name=dyn_skirt_mid_02, bone2Name=dyn_skirt_mid_01, swing1LimitDegrees=60, swing2LimitDegrees=60, twistLimitDegrees=2}
 {
  bone1Name=dyn_skirt_mid_03_l
  bone2Name=dyn_skirt_mid_02
  swing1LimitDegrees=60
  swing2LimitDegrees=45
  swing2Motion=Free
  twistLimitDegrees=30
 }
 {
  bone1Name=dyn_skirt_mid_04_l
  bone2Name=dyn_skirt_mid_03_l
  swing1LimitDegrees=60
  swing2LimitDegrees=45
  swing2Motion=Free
  twistLimitDegrees=30
 }
 {
  bone1Name=dyn_skirt_mid_05_l
  bone2Name=dyn_skirt_mid_04_l
  swing1LimitDegrees=60
  swing2LimitDegrees=45
  swing2Motion=Free
  twistLimitDegrees=30
 }
 {
  bone1Name=dyn_skirt_mid_03_r
  bone2Name=dyn_skirt_mid_02
  swing1LimitDegrees=60
  swing2LimitDegrees=45
  swing2Motion=Free
  twistLimitDegrees=30
 }
 {
  bone1Name=dyn_skirt_mid_04_r
  bone2Name=dyn_skirt_mid_03_r
  swing1LimitDegrees=70
  swing2LimitDegrees=70
  swing2Motion=Free
  twistLimitDegrees=30
 }
 {
  bone1Name=dyn_skirt_mid_05_r
  bone2Name=dyn_skirt_mid_04_r
  swing1LimitDegrees=60
  swing2LimitDegrees=45
  swing2Motion=Free
  twistLimitDegrees=30
 }
 {bone1Name=dyn_skirt_frdiag_01_l, bone2Name=pelvis, swing1LimitDegrees=30, swing2LimitDegrees=60, twistLimitDegrees=50}
 {bone1Name=dyn_skirt_frdiag_02_l, bone2Name=dyn_skirt_frdiag_01_l, swing1LimitDegrees=60, swing2LimitDegrees=60, twistLimitDegrees=2}
 {bone1Name=dyn_skirt_frdiag_03_l, bone2Name=dyn_skirt_frdiag_02_l, swing1LimitDegrees=60, swing2LimitDegrees=60, twistLimitDegrees=2}
 {bone1Name=dyn_skirt_frdiag_04_l, bone2Name=dyn_skirt_frdiag_03_l, swing1LimitDegrees=60, swing2LimitDegrees=60, twistLimitDegrees=2}
 {bone1Name=dyn_skirt_bk_01_l, bone2Name=pelvis, swing1LimitDegrees=10, swing2LimitDegrees=60, twistLimitDegrees=2}
 {bone1Name=dyn_skirt_bk_02_l, bone2Name=dyn_skirt_bk_01_l, swing1LimitDegrees=60, swing2LimitDegrees=60, twistLimitDegrees=2}
 {
  bone1Name=dyn_skirt_bk_03_l
  bone2Name=dyn_skirt_bk_02_l
  swing1LimitDegrees=45
  swing2LimitDegrees=45
  swing2Motion=Free
  twistLimitDegrees=10
 }
 {
  bone1Name=dyn_skirt_bk_04_l
  bone2Name=dyn_skirt_bk_03_l
  swing1LimitDegrees=80
  swing2LimitDegrees=45
  swing2Motion=Free
  twistLimitDegrees=30
 }
 {bone1Name=dyn_skirt_out_01_l, bone2Name=pelvis, swing1LimitDegrees=10, swing2LimitDegrees=60, twistLimitDegrees=2}
 {bone1Name=dyn_skirt_out_02_l, bone2Name=dyn_skirt_out_01_l, swing1LimitDegrees=60, swing2LimitDegrees=60, twistLimitDegrees=2}
 {bone1Name=dyn_skirt_out_03_l, bone2Name=dyn_skirt_out_02_l, swing1LimitDegrees=60, swing2LimitDegrees=60, twistLimitDegrees=2}
 {bone1Name=dyn_skirt_out_04_l, bone2Name=dyn_skirt_out_03_l, swing1LimitDegrees=60, swing2LimitDegrees=60, twistLimitDegrees=2}
 {bone1Name=dyn_skirt_frdiag_01_r, bone2Name=pelvis, swing1LimitDegrees=10, swing2LimitDegrees=60, twistLimitDegrees=2}
 {bone1Name=dyn_skirt_frdiag_02_r, bone2Name=dyn_skirt_frdiag_01_r, swing1LimitDegrees=80, swing2LimitDegrees=80, twistLimitDegrees=2}
 {bone1Name=dyn_skirt_frdiag_03_r, bone2Name=dyn_skirt_frdiag_02_r, swing1LimitDegrees=80, swing2LimitDegrees=80, twistLimitDegrees=2}
 {bone1Name=dyn_skirt_frdiag_04_r, bone2Name=dyn_skirt_frdiag_03_r, swing1LimitDegrees=90, swing2LimitDegrees=90, twistLimitDegrees=30}
 {bone1Name=dyn_skirt_bk_01_r, bone2Name=pelvis, swing1LimitDegrees=60, swing2LimitDegrees=60, twistLimitDegrees=2}
 {bone1Name=dyn_skirt_bk_02_r, bone2Name=dyn_skirt_bk_01_r, swing1LimitDegrees=60, swing2LimitDegrees=60, twistLimitDegrees=2}
 {
  bone1Name=dyn_skirt_bk_03_r
  bone2Name=dyn_skirt_bk_02_r
  swing1LimitDegrees=45
  swing2LimitDegrees=45
  swing2Motion=Free
  twistLimitDegrees=10
 }
 {
  bone1Name=dyn_skirt_bk_04_r
  bone2Name=dyn_skirt_bk_03_r
  swing1LimitDegrees=45
  swing2LimitDegrees=45
  swing2Motion=Free
  twistLimitDegrees=10
 }
 {bone1Name=dyn_skirt_out_01_r, bone2Name=pelvis, swing1LimitDegrees=10, swing2LimitDegrees=60, twistLimitDegrees=2}
 {bone1Name=dyn_skirt_out_02_r, bone2Name=dyn_skirt_out_01_r, swing1LimitDegrees=60, swing2LimitDegrees=60, twistLimitDegrees=2}
 {bone1Name=dyn_skirt_out_03_r, bone2Name=dyn_skirt_out_02_r, swing1LimitDegrees=60, swing2LimitDegrees=60, twistLimitDegrees=2}
 {bone1Name=dyn_skirt_out_04_r, bone2Name=dyn_skirt_out_03_r, swing1LimitDegrees=60, swing2LimitDegrees=60, twistLimitDegrees=2}
 {
  bone1Name=dyn_skirt_out_04_r
  bone2Name=dyn_skirt_bk_04_r
  swing1LimitDegrees=45
  swing2LimitDegrees=45
  swing2Motion=Free
  twistLimitDegrees=10
 }
 {
  bone1Name=dyn_skirt_mid_05_r
  bone2Name=dyn_skirt_bk_04_r
  swing1LimitDegrees=60
  swing2LimitDegrees=45
  swing2Motion=Free
  twistLimitDegrees=30
 }
 {
  bone1Name=dyn_skirt_mid_05_l
  bone2Name=dyn_skirt_bk_04_l
  swing1LimitDegrees=60
  swing2LimitDegrees=45
  swing2Motion=Free
  twistLimitDegrees=30
 }
 {
  bone1Name=dyn_skirt_out_04_l
  bone2Name=dyn_skirt_bk_04_l
  swing1LimitDegrees=45
  swing2LimitDegrees=45
  swing2Motion=Free
  twistLimitDegrees=10
 }
 {bone1Name=dyn_skirt_fr_05_r, bone2Name=dyn_skirt_frdiag_04_r, swing1LimitDegrees=135, swing2LimitDegrees=135, twistLimitDegrees=10}
 {bone1Name=dyn_skirt_fr_05_l, bone2Name=dyn_skirt_frdiag_04_l, swing1LimitDegrees=120, swing2LimitDegrees=130, twistLimitDegrees=30}
 {
  bone1Name=pelvis
  bone2Name=root
  swing1LimitDegrees=45
  swing1Motion=Free
  swing2LimitDegrees=45
  swing2Motion=Free
  twistLimitDegrees=45
  twistMotion=Free
 }
 {
  bone1Name=dyn_skirt_fr_02a_l
  bone2Name=dyn_skirt_fr_02_l
  swing1LimitDegrees=60
  swing1Motion=Free
  swing2LimitDegrees=135
  swing2Motion=Free
  twistLimitDegrees=10
 }
 {bone1Name=dyn_skirt_fr_02a_r, bone2Name=dyn_skirt_fr_02_r, swing1LimitDegrees=90, swing2LimitDegrees=90, twistLimitDegrees=30}
 {
  bone1Name=dyn_skirt_bk_03_r
  bone2Name=dyn_skirt_mid_04_r
  swing1LimitDegrees=45
  swing1Motion=Free
  swing2LimitDegrees=45
  swing2Motion=Free
  twistLimitDegrees=10
  twistMotion=Free
 }
 {
  bone1Name=foot_r
  bone2Name=calf_r
  swing1LimitDegrees=45
  swing1Motion=Free
  swing2LimitDegrees=45
  swing2Motion=Free
  twistLimitDegrees=45
  twistMotion=Free
 }
 {
  bone1Name=foot_l
  bone2Name=calf_l
  swing1LimitDegrees=45
  swing1Motion=Free
  swing2LimitDegrees=45
  swing2Motion=Free
  twistLimitDegrees=45
  twistMotion=Free
 }
 {bone1Name=dyn_skirt_fr_03_l, bone2Name=dyn_skirt_frdiag_02_l, swing1LimitDegrees=50, swing2LimitDegrees=135, twistLimitDegrees=10}
 {
  bone1Name=dyn_skirt_frdiag_04_l
  bone2Name=dyn_skirt_out_04_l
  swing1LimitDegrees=120
  swing1Motion=Free
  swing2LimitDegrees=130
  swing2Motion=Free
  twistLimitDegrees=30
  twistMotion=Free
 }
]
