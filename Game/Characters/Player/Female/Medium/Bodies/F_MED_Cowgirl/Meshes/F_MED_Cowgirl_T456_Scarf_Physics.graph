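# Physics asset for the scarf on the female medium Cowgirl player body (T456), defining its dynamic bone bodies and joints.
# Capsules on the spine, left clavicle and left upper arm anchor limited-angle scarf chains down the arm and chest, some ending in tassels.
# The right chest piece is a single box, and the arm root is locked in rotation to the upper arm.
# A free-rotation constraint cross-links the left chest and front arm chains.
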PhysicsAsset /Game/Characters/Player/Female/Medium/Bodies/F_MED_Cowgirl/Meshes/F_MED_Cowgirl_T456_Scarf_Physics
bodies {
 clavicle_l [{shapeType=Capsule, length=2.0826, radius=2.7179, rotation=<rot 90 0 0>}]
 dyn_scarf_arm_01 [{shapeType=Capsule, length=4.5, radius=4, shapeName=dyn_scarf_arm_01_capsule, center=<2.138 -0.0253 2.5903>, rotation=<rot 0 0 89.438>}]
 dyn_scarf_arm_02 [{shapeType=Capsule, length=4.5, radius=4, shapeName=dyn_scarf_arm_02_capsule, center=<1.7828 -0.0253 2.5902>, rotation=<rot 0 0 89.438>}]
 dyn_scarf_arm_03 [{shapeType=Capsule, length=6.25, radius=2.25, shapeName=dyn_scarf_arm_03_capsule, center=<0 -0.0146 1.5104>, rotation=<rot 0 0 89.438>}]
 dyn_scarf_arm_bk_01 [{
   shapeType=Capsule
   length=5.75
   radius=3.75
   shapeName=dyn_scarf_arm_bk_01_capsule
   center=<3.4888 0.0248 1.1066>
   rotation=<rot 0 0 -88.7154>
  }]
 dyn_scarf_arm_bk_02 [{shapeType=Capsule, length=6.5, radius=3, shapeName=dyn_scarf_arm_bk_02_capsule, center=<2.0677 0.0237 1.0557>, rotation=<rot 0 0 -88.7154>}]
 dyn_scarf_arm_bk_03 [{
   shapeType=Capsule
   length=5.75
   radius=2.5
   shapeName=dyn_scarf_arm_bk_03_capsule
   center=<0.5031 0.0137 0.6117>
   rotation=<rot 0 0 -88.7154>
  }]
 dyn_scarf_arm_bk_tassels [{
   shapeType=Capsule
   length=3.75
   radius=2.5
   shapeName=dyn_scarf_arm_bk_tassels_capsule
   center=<3.44 0.0148 0.6604>
   rotation=<rot 0 0 -88.7154>
  }]
 dyn_scarf_arm_fr_01 [{
   shapeType=Capsule
   length=4.5
   radius=2
   shapeName=dyn_scarf_arm_fr_01_capsule
   center=<2.9373 -0.0648 0.1221>
   rotation=<rot -87.6168 -1.2602 -0.0028>
  }]
 dyn_scarf_arm_fr_02 [{
   shapeType=Capsule
   length=4
   radius=2
   shapeName=dyn_scarf_arm_fr_02_capsule
   center=<2.0798 -0.0459 0.0863>
   rotation=<rot -87.6168 -1.2602 -0.0028>
  }]
 dyn_scarf_arm_fr_03 [{
   shapeType=Capsule
   length=4
   radius=2
   shapeName=dyn_scarf_arm_fr_03_capsule
   center=<1.7415 -0.0384 0.0724>
   rotation=<rot -87.6167 -1.2602 -0.0028>
  }]
 dyn_scarf_arm_tassels [{shapeType=Capsule, length=6, radius=2.5, shapeName=dyn_scarf_arm_tassels_capsule, center=<3.0492 -0.016 1.6254>, rotation=<rot 0 0 89.438>}]
 dyn_scarf_chest_bk_01 [{
   shapeType=Capsule
   length=3
   radius=3
   shapeName=dyn_scarf_chest_bk_01_capsule
   center=<2.3219 -0.0067 1.1347>
   rotation=<rot -0.0001 0 -90.3351>
  }]
 dyn_scarf_chest_bk_02 [{
   shapeType=Capsule
   length=1.75
   radius=3
   shapeName=dyn_scarf_chest_bk_02_capsule
   center=<0 -0.006 1.033>
   rotation=<rot -0.0001 0 -90.3351>
  }]
 dyn_scarf_chest_bk_tassels [{
   shapeType=Capsule
   length=0
   radius=3
   shapeName=dyn_scarf_chest_bk_tassels_capsule
   center=<2.4377 -0.0083 1.412>
   rotation=<rot -0.0001 0 -90.3351>
  }]
 dyn_scarf_chest_l_01 [{
   shapeType=Capsule
   length=7.75
   radius=2.25
   shapeName=dyn_scarf_chest_l_01_capsule
   center=<3.0295 -0.0671 0.1299>
   rotation=<rot -87.5461 -1.2662 -0.0018>
  }]
 dyn_scarf_chest_l_02 [{
   shapeType=Capsule
   length=6.5
   radius=2.25
   shapeName=dyn_scarf_chest_l_02_capsule
   center=<3.0295 -0.0671 0.1299>
   rotation=<rot -87.5461 -1.2662 -0.0018>
  }]
 dyn_scarf_chest_r_01 [{shapeType=Box, extentX=7.2601, extentY=6.0798, extentZ=0.9477, center=<4.1134 -0.0001 0.0001>}]
 spine_05 [
  {shapeType=Capsule, length=5.8786, radius=11.1398, center=<-4.5535 -1.3738 -5.1706>}
  {shapeType=Capsule, length=11.0014, radius=6.0441, center=<-7.4096 8.7152 -5.5187>, rotation=<rot 0 0 15.0039>}
 ]
 upperarm_twist_01_l [{shapeType=Capsule, length=19.9869, radius=6.845, center=<7.2971 0.8979 -2.4229>, rotation=<rot 90 0.0264 0.0264>}]
}
constraints [
 {
  bone1Name=dyn_scarf_arm_fr_01
  bone2Name=upperarm_twist_01_l
  swing1LimitDegrees=5
  swing1Motion=Locked
  swing2LimitDegrees=30
  twistLimitDegrees=20
 }
 {
  bone1Name=clavicle_l
  bone2Name=spine_05
  swing1LimitDegrees=45
  swing1Motion=Free
  swing2LimitDegrees=45
  swing2Motion=Free
  twistLimitDegrees=45
  twistMotion=Free
 }
 {
  bone1Name=dyn_scarf_arm_fr_02
  bone2Name=dyn_scarf_arm_fr_01
  swing1LimitDegrees=5
  swing1Motion=Locked
  swing2LimitDegrees=30
  twistLimitDegrees=20
 }
 {
  bone1Name=dyn_scarf_arm_fr_03
  bone2Name=dyn_scarf_arm_fr_02
  swing1LimitDegrees=5
  swing1Motion=Locked
  swing2LimitDegrees=30
  twistLimitDegrees=20
 }
 {bone1Name=dyn_scarf_arm_bk_01, bone2Name=clavicle_l, swing1LimitDegrees=10, swing2LimitDegrees=20, twistLimitDegrees=5}
 {bone1Name=dyn_scarf_arm_bk_02, bone2Name=dyn_scarf_arm_bk_01, swing1LimitDegrees=60, swing2LimitDegrees=30, twistLimitDegrees=5}
 {bone1Name=dyn_scarf_arm_bk_03, bone2Name=dyn_scarf_arm_bk_02, swing1LimitDegrees=60, swing2LimitDegrees=60, twistLimitDegrees=5}
 {
  bone1Name=dyn_scarf_arm_01
  bone2Name=upperarm_twist_01_l
  swing1LimitDegrees=5
  swing1Motion=Locked
  swing2LimitDegrees=30
  swing2Motion=Locked
  twistLimitDegrees=10
  twistMotion=Locked
 }
 {bone1Name=dyn_scarf_chest_bk_01, bone2Name=spine_05, swing1LimitDegrees=20, swing2LimitDegrees=45, twistLimitDegrees=5}
 {bone1Name=dyn_scarf_chest_bk_02, bone2Name=dyn_scarf_chest_bk_01, swing1LimitDegrees=45, swing2LimitDegrees=45, twistLimitDegrees=1}
 {bone1Name=dyn_scarf_chest_l_01, bone2Name=clavicle_l, swing1LimitDegrees=45, swing2LimitDegrees=45, twistLimitDegrees=10}
 {bone1Name=dyn_scarf_chest_l_02, bone2Name=dyn_scarf_chest_l_01, swing1LimitDegrees=45, swing2LimitDegrees=45, twistLimitDegrees=10}
 {bone1Name=dyn_scarf_chest_r_01, bone2Name=spine_05, swing1LimitDegrees=30, swing2LimitDegrees=60, twistLimitDegrees=5}
 {
  bone1Name=upperarm_twist_01_l
  bone2Name=clavicle_l
  swing1LimitDegrees=45
  swing1Motion=Free
  swing2LimitDegrees=45
  swing2Motion=Free
  twistLimitDegrees=45
  twistMotion=Free
 }
 {
  bone1Name=dyn_scarf_arm_fr_02
  bone2Name=dyn_scarf_chest_l_02
  swing1LimitDegrees=45
  swing1Motion=Free
  swing2LimitDegrees=45
  swing2Motion=Free
  twistLimitDegrees=45
  twistMotion=Free
 }
 {bone1Name=dyn_scarf_arm_02, bone2Name=dyn_scarf_arm_01, swing1LimitDegrees=5, swing2LimitDegrees=30, twistLimitDegrees=10}
 {bone1Name=dyn_scarf_arm_03, bone2Name=dyn_scarf_arm_02, swing1LimitDegrees=5, swing2LimitDegrees=45, twistLimitDegrees=10}
 {bone1Name=dyn_scarf_chest_bk_tassels, bone2Name=dyn_scarf_chest_bk_02, swing1LimitDegrees=45, swing2LimitDegrees=45, twistLimitDegrees=1}
 {bone1Name=dyn_scarf_arm_bk_tassels, bone2Name=dyn_scarf_arm_bk_03, swing1LimitDegrees=45, swing2LimitDegrees=45, twistLimitDegrees=5}
 {bone1Name=dyn_scarf_arm_tassels, bone2Name=dyn_scarf_arm_03, swing1LimitDegrees=20, swing2LimitDegrees=30, twistLimitDegrees=10}
]
