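# Physics asset for the medium female Cowgirl player body, setting up physics for the left and right spurs on the feet.
# Each dynamic spur bone has a tiny, near-spherical capsule constrained to a capsule on its foot bone.
# Both constraints lock swing and free twist, so each spur can only rotate about one axis.
# Despite the Tassels name, it holds no tassel bodies.
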PhysicsAsset /Game/Characters/Player/Female/Medium/Bodies/F_MED_Cowgirl/Meshes/F_MED_Cowgirl_T456_Tassels_Physics
bodies {
 dyn_spur_l [{shapeType=Capsule, length=0.0001, radius=2, center=<0 0.05 0.05>}]
 dyn_spur_r [{shapeType=Capsule, length=0.0001, radius=2, center=<0 0.05 0.05>}]
 foot_l [{shapeType=Capsule, length=2.0826, radius=2.7179, rotation=<rot 90 0 0>}]
 foot_r [{shapeType=Capsule, length=2.0826, radius=2.7179, rotation=<rot 90 0 0>}]
}
constraints [
 {
  bone1Name=dyn_spur_l
  bone2Name=foot_l
  swing1LimitDegrees=45
  swing1Motion=Locked
  swing2LimitDegrees=45
  swing2Motion=Locked
  twistLimitDegrees=45
  twistMotion=Free
 }
 {
  bone1Name=dyn_spur_r
  bone2Name=foot_r
  swing1LimitDegrees=45
  swing1Motion=Locked
  swing2LimitDegrees=45
  swing2Motion=Locked
  twistLimitDegrees=45
  twistMotion=Free
 }
]
