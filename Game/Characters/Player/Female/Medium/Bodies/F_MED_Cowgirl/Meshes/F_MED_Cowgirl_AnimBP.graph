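# Animation Blueprint for the F_MED_Cowgirl body that copies the attached parent's pose and adds secondary motion to skirt, scarf, coat, tassels and straps.
# Bone constraints and crouch-weighted bone-driven controllers shape the skirt, with some calf constraints switched by bRigidBodyEnabled.
# RigidBody nodes simulate scarf, coat and tassel physics assets, with skydiving force on the first two.
# Constrained AnimDynamics bones follow, then skirt bones shrink by an in-vehicle alpha.
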AnimBlueprint /Game/Characters/Player/Female/Medium/Bodies/F_MED_Cowgirl/Meshes/F_MED_Cowgirl_AnimBP
target_skeleton=&/F_MED_Cowgirl_Skeleton
graphs [
 AnimGraph {nodes [
   n1 = "Constraint - dyn_skirt_fr_03_r (R)"!(ComponentPose=Local To Component(LocalPose=Use cached pose 'SavedPose13205'()), Alpha=0.25) -> (Pose) {
    alphaBoolBlend {blendOption=Linear}
    boneToModify {boneName=dyn_skirt_fr_03_r}
    constraintSetup [{transformType=Rotation, targetBone {boneName=calf_r}}]
    constraintWeights [1]
   }
   n2 = "Constraint - dyn_skirt_fr_04_r (R)"!(ComponentPose=@n1, Alpha=1) -> (Pose) {
    alphaBoolBlend {blendOption=Linear}
    boneToModify {boneName=dyn_skirt_fr_04_r}
    constraintSetup [{transformType=Rotation, targetBone {boneName=calf_r}}]
    constraintWeights [1]
   }
   n3 = Component To Local(ComponentPose=@n2)
   n4 = Use cached pose 'SavedPose13205'()
   n5 = NOT Boolean(A=Get bRigidBodyEnabled())
   n6 = "Constraint - dyn_scarf_arm_01 (PP)"!(
    ComponentPose=Local To Component(LocalPose=Blend Poses by bool(BlendPose_0=@n3, BlendPose_1=@n4, bActiveValue=@n5) {blendType=Linear})
    Alpha=1
    ConstraintWeights_0=0.7179
    ConstraintWeights_1:float=0.2821
   ) -> (Pose) {
    alphaBoolBlend {blendOption=Linear}
    boneToModify {boneName=dyn_scarf_arm_01}
    constraintSetup [{targetBone {boneName=upperarm_twist_01_l}}, {targetBone {boneName=clavicle_l}}]
    constraintWeights [1,1]
   }
   n7 = "Constraint - dyn_scarf_arm_fr_01 (PPP)"!(ComponentPose=@n6, Alpha=1, ConstraintWeights_0=0.1265, ConstraintWeights_1:float=0.6988, ConstraintWeights_2:float=0.1747) -> (Pose) {
    alphaBoolBlend {blendOption=Linear}
    boneToModify {boneName=dyn_scarf_arm_fr_01}
    constraintSetup [{targetBone {boneName=upperarm_twist_01_l}}, {targetBone {boneName=clavicle_l}}, {targetBone {boneName=spine_05}}]
    constraintWeights [1,1,1]
   }
   n8 = "Constraint - dyn_scarf_arm_bk_01 (PP)"!(ComponentPose=@n7, Alpha=1, ConstraintWeights_0=0.2441, ConstraintWeights_1:float=0.7559) -> (Pose) {
    alphaBoolBlend {blendOption=Linear}
    boneToModify {boneName=dyn_scarf_arm_bk_01}
    constraintSetup [{targetBone {boneName=upperarm_twist_01_l}}, {targetBone {boneName=clavicle_l}}]
    constraintWeights [1,1]
   }
   n9 = "Constraint - dyn_scarf_arm_fr_01 (R)"(ComponentPose=@n8, Alpha=1, ConstraintWeights_0=1) {
    alphaBoolBlend {blendOption=Linear}
    boneToModify {boneName=dyn_scarf_arm_fr_01}
    constraintSetup [{transformType=Rotation, targetBone {boneName=upperarm_twist_01_l}}]
    constraintWeights [1]
   }
   n10 = "Constraint - dyn_scarf_chest_l_01 (RR)"!(ComponentPose=@n9, Alpha=1, ConstraintWeights_0=1, ConstraintWeights_1:float=1) -> (Pose) {
    alphaBoolBlend {blendOption=Linear}
    boneToModify {boneName=dyn_scarf_chest_l_01}
    constraintSetup [{transformType=Rotation, targetBone {boneName=dyn_scarf_arm_fr_01}}, {transformType=Rotation, targetBone {boneName=dyn_scarf_chest_r_01}}]
    constraintWeights [1,1]
   }
   n11 = Break Fort Anim Input Skydiving External Force(FortAnimInput_SkydivingExternalForce=Get SkydivingExternalForceInput()) -> (
    bUseSkydivingVectorForce=~
    SkydivingVectorForce=~
    HeadToPelvisDirection=~
    FloatingMultiplier=~
    FloatingAdditive=~
    DivingMultiplier=~
    DivingAdditive=~
    ParachutingMultiplier=~
    ParachutingAdditive=~
   )
   n12 = RigidBody(ComponentPose=@n10, Alpha=Get ScarfRbanAlphaFINAL() -> (ScarfRbanAlphaFINAL:double), ExternalForce=@n11.SkydivingVectorForce) {
    bClampLinearTranslationLimitToRefPose=true
    bOverrideWorldGravity=true
    overridePhysicsAsset=&/F_MED_Cowgirl_T456_Scarf_Physics
    simulationSpace=BaseBoneSpace
    alphaBoolBlend {blendOption=Linear}
    baseBoneRef {boneName=spine_05}
    componentAppliedLinearAccClamp=<3000 5000 5000>
    componentLinearAccScale=<0.2 0.5 0.5>
    componentLinearVelScale=<0.2 0.5 0.5>
    overrideWorldGravity {z=-980}
   }
   n13 = RigidBody(ComponentPose=@n12, Alpha=Get CoatRbanAlphaFINAL() -> (CoatRbanAlphaFINAL:double), ExternalForce=@n11.SkydivingVectorForce) {
    bClampLinearTranslationLimitToRefPose=true
    bOverrideWorldGravity=true
    overridePhysicsAsset=&/F_MED_Cowgirl_T456_Coat_Physics
    simulationSpace=BaseBoneSpace
    alphaBoolBlend {blendOption=Linear}
    baseBoneRef {boneName=root}
    componentAppliedLinearAccClamp=<5000 5000 5000>
    componentLinearAccScale=<0.01 0.01 0.01>
    componentLinearVelScale=<0.01 0.01 0.01>
    overrideWorldGravity {z=-980}
   }
   n14 = RigidBody(ComponentPose=@n13, Alpha=Get SpursRbanAlpha() -> (SpursRbanAlpha:double)) {
    bClampLinearTranslationLimitToRefPose=true
    bOverrideWorldGravity=true
    overridePhysicsAsset=&/F_MED_Cowgirl_T456_Tassels_Physics
    simulationSpace=BaseBoneSpace
    alphaBoolBlend {blendOption=Linear}
    baseBoneRef {boneName=root}
    componentAppliedLinearAccClamp=<1500 1500 1500>
    componentLinearAccScale=<0.1 0.1 0.5>
    componentLinearVelScale=<0.1 0.1 0.5>
    overrideWorldGravity {z=-980}
   }
   n15 = Get LegTasselsAnimDynAlpha() -> (LegTasselsAnimDynAlpha:double)
   n16 = "Constraint - dyn_tassel_legs_06 (PP)"!(ComponentPose=@n14, Alpha=@n15, ConstraintWeights_0=0.807, ConstraintWeights_1:float=0.193) -> (Pose) {
    alphaBoolBlend {blendOption=Linear}
    boneToModify {boneName=dyn_tassel_legs_06}
    constraintSetup [{targetBone {boneName=pelvis}}, {targetBone {boneName=thigh_twist_01_l}}]
    constraintWeights [1,1]
   }
   n17 = "Constraint - dyn_tassel_legs_07 (P)"!(ComponentPose=@n16, Alpha=@n15, ConstraintWeights_0=1, ConstraintWeights_1:float=0) -> (Pose) {
    alphaBoolBlend {blendOption=Linear}
    boneToModify {boneName=dyn_tassel_legs_07}
    constraintSetup [{targetBone {boneName=pelvis}}]
    constraintWeights [1,0]
   }
   n18 = "Constraint - dyn_tassel_legs_08 (P)"!(ComponentPose=@n17, Alpha=@n15, ConstraintWeights_0=1, ConstraintWeights_1:float=0) -> (Pose) {
    alphaBoolBlend {blendOption=Linear}
    boneToModify {boneName=dyn_tassel_legs_08}
    constraintSetup [{targetBone {boneName=pelvis}}]
    constraintWeights [1,0]
   }
   n19 = "Constraint - dyn_tassel_legs_09 (PP)"!(ComponentPose=@n18, Alpha=@n15, ConstraintWeights_0=0.9543, ConstraintWeights_1:float=0.0457) -> (Pose) {
    alphaBoolBlend {blendOption=Linear}
    boneToModify {boneName=dyn_tassel_legs_09}
    constraintSetup [{targetBone {boneName=pelvis}}, {targetBone {boneName=thigh_twist_01_r}}]
    constraintWeights [1,1]
   }
   n20 = "Constraint - dyn_tassel_legs_015 (PP)"!(ComponentPose=@n19, Alpha=@n15, ConstraintWeights_0=0.9543, ConstraintWeights_1:float=0.0457) -> (Pose) {
    alphaBoolBlend {blendOption=Linear}
    boneToModify {boneName=dyn_tassel_legs_015}
    constraintSetup [{targetBone {boneName=pelvis}}, {targetBone {boneName=thigh_twist_01_l}}]
    constraintWeights [1,1]
   }
   n21 = "Constraint - dyn_tassel_legs_016 (P)"!(ComponentPose=@n20, Alpha=@n15, ConstraintWeights_0=1, ConstraintWeights_1:float=0) -> (Pose) {
    alphaBoolBlend {blendOption=Linear}
    boneToModify {boneName=dyn_tassel_legs_016}
    constraintSetup [{targetBone {boneName=pelvis}}]
    constraintWeights [1,0]
   }
   n22 = "Constraint - dyn_tassel_legs_017 (PP)"!(ComponentPose=@n21, Alpha=@n15, ConstraintWeights_0=0.9952, ConstraintWeights_1:float=0.0048) -> (Pose) {
    alphaBoolBlend {blendOption=Linear}
    boneToModify {boneName=dyn_tassel_legs_017}
    constraintSetup [{targetBone {boneName=pelvis}}, {targetBone {boneName=thigh_twist_01_r}}]
    constraintWeights [1,1]
   }
   n23 = "Constraint - dyn_tassel_legs_018 (PP)"!(ComponentPose=@n22, Alpha=@n15, ConstraintWeights_0=0.9841, ConstraintWeights_1:float=0.0159) -> (Pose) {
    alphaBoolBlend {blendOption=Linear}
    boneToModify {boneName=dyn_tassel_legs_018}
    constraintSetup [{targetBone {boneName=pelvis}}, {targetBone {boneName=thigh_twist_01_r}}]
    constraintWeights [1,1]
   }
   n24 = "Constraint - dyn_tassel_legs_012 (P)"!(ComponentPose=@n23, Alpha=@n15, ConstraintWeights_0=1, ConstraintWeights_1:float=0) -> (Pose) {
    alphaBoolBlend {blendOption=Linear}
    boneToModify {boneName=dyn_tassel_legs_012}
    constraintSetup [{targetBone {boneName=thigh_r}}]
    constraintWeights [1,0]
   }
   n25 = "Constraint - dyn_tassel_legs_011 (PP)"!(ComponentPose=@n24, Alpha=@n15, ConstraintWeights_0=0.7896, ConstraintWeights_1:float=0.2104) -> (Pose) {
    alphaBoolBlend {blendOption=Linear}
    boneToModify {boneName=dyn_tassel_legs_011}
    constraintSetup [{targetBone {boneName=thigh_r}}, {targetBone {boneName=thigh_twist_01_r}}]
    constraintWeights [1,1]
   }
   n26 = "Constraint - dyn_tassel_legs_010 (PPP)"!(ComponentPose=@n25, Alpha=@n15, ConstraintWeights_0=0.5365, ConstraintWeights_1:float=0.2407, ConstraintWeights_2:float=0.2228) -> (Pose) {
    alphaBoolBlend {blendOption=Linear}
    boneToModify {boneName=dyn_tassel_legs_010}
    constraintSetup [{targetBone {boneName=thigh_twist_01_r}}, {targetBone {boneName=thigh_r}}, {targetBone {boneName=pelvis}}]
    constraintWeights [1,1,1]
   }
   n27 = "Constraint - dyn_tassel_legs_014 (PP)"!(ComponentPose=@n26, Alpha=@n15, ConstraintWeights_0=0.9189, ConstraintWeights_1:float=0.0811) -> (Pose) {
    alphaBoolBlend {blendOption=Linear}
    boneToModify {boneName=dyn_tassel_legs_014}
    constraintSetup [{targetBone {boneName=calf_r}}, {targetBone {boneName=calf_twist_02_r}}]
    constraintWeights [1,1]
   }
   n28 = "Constraint - dyn_tassel_legs_013 (PP)"!(ComponentPose=@n27, Alpha=@n15, ConstraintWeights_0=0.8935, ConstraintWeights_1:float=0.1065) -> (Pose) {
    alphaBoolBlend {blendOption=Linear}
    boneToModify {boneName=dyn_tassel_legs_013}
    constraintSetup [{targetBone {boneName=thigh_r}}, {targetBone {boneName=calf_r}}]
    constraintWeights [1,1]
   }
   n29 = "Constraint - dyn_tassel_legs_05 (PPP)"!(ComponentPose=@n28, Alpha=@n15, ConstraintWeights_0=0.5365, ConstraintWeights_1:float=0.2407, ConstraintWeights_2:float=0.2228) -> (Pose) {
    alphaBoolBlend {blendOption=Linear}
    boneToModify {boneName=dyn_tassel_legs_05}
    constraintSetup [{targetBone {boneName=thigh_twist_01_l}}, {targetBone {boneName=thigh_l}}, {targetBone {boneName=pelvis}}]
    constraintWeights [1,1,1]
   }
   n30 = "Constraint - dyn_tassel_legs_04 (PPP)"!(ComponentPose=@n29, Alpha=@n15, ConstraintWeights_0=0.598, ConstraintWeights_1:float=0.3707, ConstraintWeights_2:float=0.0313) -> (Pose) {
    alphaBoolBlend {blendOption=Linear}
    boneToModify {boneName=dyn_tassel_legs_04}
    constraintSetup [{targetBone {boneName=thigh_l}}, {targetBone {boneName=thigh_twist_01_l}}, {targetBone {boneName=pelvis}}]
    constraintWeights [1,1,1]
   }
   n31 = "Constraint - dyn_tassel_legs_03 (PPP)"!(ComponentPose=@n30, Alpha=@n15, ConstraintWeights_0=0.7958, ConstraintWeights_1:float=0.2042, ConstraintWeights_2:float=0) -> (Pose) {
    alphaBoolBlend {blendOption=Linear}
    boneToModify {boneName=dyn_tassel_legs_03}
    constraintSetup [{targetBone {boneName=thigh_l}}, {targetBone {boneName=thigh_twist_01_l}}, {targetBone {boneName=pelvis}}]
    constraintWeights [1,1,1]
   }
   n32 = "Constraint - dyn_tassel_legs_02 (P)"!(ComponentPose=@n31, Alpha=@n15, ConstraintWeights_0=1, ConstraintWeights_1:float=0) -> (Pose) {
    alphaBoolBlend {blendOption=Linear}
    boneToModify {boneName=dyn_tassel_legs_02}
    constraintSetup [{targetBone {boneName=thigh_l}}]
    constraintWeights [1,0]
   }
   n33 = "Constraint - dyn_tassel_legs_01 (PP)"!(ComponentPose=@n32, Alpha=@n15, ConstraintWeights_0=0.593, ConstraintWeights_1:float=0.407) -> (Pose) {
    alphaBoolBlend {blendOption=Linear}
    boneToModify {boneName=dyn_tassel_legs_01}
    constraintSetup [{targetBone {boneName=thigh_l}}, {targetBone {boneName=calf_l}}]
    constraintWeights [1,1]
   }
   n34 = Anim Dynamics - Bone: dyn_tassel_legs_01(ComponentPose=@n33, Alpha=@n15, ExternalForce="0,0,0") {
    angularDampingOverride=0.9999
    angularSpringConstant=12000
    bAngularSpring=true
    bGravityOverrideInSimSpace=true
    bOverrideAngularDamping=true
    bUseGravityOverride=true
    gravityScale=0.35
    simulationSpace=RootRelative
    alphaBoolBlend {blendOption=Linear}
    boundBone {boneName=dyn_tassel_legs_01}
    componentAppliedLinearAccClamp=<3000 3000 3000>
    componentLinearAccScale=<0.4 0.4 0.4>
    componentLinearVelScale=<0.4 0.4 0.4>
    gravityOverride {z=-980}
    physicsBodyDefinitions [{
      sphereCollisionRadius=0
      boundBone {boneName=dyn_tassel_legs_01}
      boxExtents=<15 18 26>
      constraintSetup {angularLimitsMax=<15 15 15>, angularLimitsMin=<-15 -15 -15>, angularTarget=<1 0 0>}
      localJointOffset=<4.5 0 0>
     }]
   }
   n35 = Anim Dynamics - Bone: dyn_tassel_legs_02(ComponentPose=@n34, Alpha=@n15, ExternalForce="0,0,0") {
    angularDampingOverride=0.9999
    angularSpringConstant=12000
    bAngularSpring=true
    bGravityOverrideInSimSpace=true
    bOverrideAngularDamping=true
    bUseGravityOverride=true
    gravityScale=0.35
    simulationSpace=RootRelative
    alphaBoolBlend {blendOption=Linear}
    boundBone {boneName=dyn_tassel_legs_02}
    componentAppliedLinearAccClamp=<3000 3000 3000>
    componentLinearAccScale=<0.4 0.4 0.4>
    componentLinearVelScale=<0.4 0.4 0.4>
    gravityOverride {z=-980}
    physicsBodyDefinitions [{
      sphereCollisionRadius=0
      boundBone {boneName=dyn_tassel_legs_02}
      boxExtents=<15 22 28>
      constraintSetup {angularLimitsMax=<15 15 15>, angularLimitsMin=<-15 -15 -15>, angularTarget=<1 0 0>}
      localJointOffset=<4.5 0 0>
     }]
   }
   n36 = Anim Dynamics - Bone: dyn_tassel_legs_03(ComponentPose=@n35, Alpha=@n15, ExternalForce="0,0,0") {
    angularDampingOverride=0.9999
    angularSpringConstant=12000
    bAngularSpring=true
    bGravityOverrideInSimSpace=true
    bOverrideAngularDamping=true
    bUseGravityOverride=true
    gravityScale=0.35
    simulationSpace=RootRelative
    alphaBoolBlend {blendOption=Linear}
    boundBone {boneName=dyn_tassel_legs_03}
    componentAppliedLinearAccClamp=<3000 3000 3000>
    componentLinearAccScale=<0.4 0.4 0.4>
    componentLinearVelScale=<0.4 0.4 0.4>
    gravityOverride {z=-980}
    physicsBodyDefinitions [{
      sphereCollisionRadius=0
      boundBone {boneName=dyn_tassel_legs_03}
      boxExtents=<15 18 26>
      constraintSetup {angularLimitsMax=<15 15 15>, angularLimitsMin=<-15 -15 -15>, angularTarget=<1 0 0>}
      localJointOffset=<4.5 0 0>
     }]
   }
   n37 = Anim Dynamics - Bone: dyn_tassel_legs_04(ComponentPose=@n36, Alpha=@n15, ExternalForce="0,0,0") {
    angularDampingOverride=0.9999
    angularSpringConstant=12000
    bAngularSpring=true
    bGravityOverrideInSimSpace=true
    bOverrideAngularDamping=true
    bUseGravityOverride=true
    gravityScale=0.35
    simulationSpace=RootRelative
    alphaBoolBlend {blendOption=Linear}
    boundBone {boneName=dyn_tassel_legs_04}
    componentAppliedLinearAccClamp=<3000 3000 3000>
    componentLinearAccScale=<0.4 0.4 0.4>
    componentLinearVelScale=<0.4 0.4 0.4>
    gravityOverride {z=-980}
    physicsBodyDefinitions [{
      sphereCollisionRadius=0
      boundBone {boneName=dyn_tassel_legs_04}
      boxExtents=<15 27 16>
      constraintSetup {angularLimitsMax=<15 15 15>, angularLimitsMin=<-15 -15 -15>, angularTarget=<1 0 0>}
      localJointOffset=<4.5 0 0>
     }]
   }
   n38 = Anim Dynamics - Bone: dyn_tassel_legs_05(ComponentPose=@n37, Alpha=@n15, ExternalForce="0,0,0") {
    angularDampingOverride=0.9999
    angularSpringConstant=12000
    bAngularSpring=true
    bGravityOverrideInSimSpace=true
    bOverrideAngularDamping=true
    bUseGravityOverride=true
    gravityScale=0.35
    simulationSpace=RootRelative
    alphaBoolBlend {blendOption=Linear}
    boundBone {boneName=dyn_tassel_legs_05}
    componentAppliedLinearAccClamp=<3000 3000 3000>
    componentLinearAccScale=<0.4 0.4 0.4>
    componentLinearVelScale=<0.4 0.4 0.4>
    gravityOverride {z=-980}
    physicsBodyDefinitions [{
      sphereCollisionRadius=0
      boundBone {boneName=dyn_tassel_legs_05}
      boxExtents=<15 18 26>
      constraintSetup {angularLimitsMax=<15 15 15>, angularLimitsMin=<-15 -15 -15>, angularTarget=<1 0 0>}
      localJointOffset=<4.5 0 0>
     }]
   }
   n39 = Anim Dynamics - Bone: dyn_tassel_legs_06(ComponentPose=@n38, Alpha=@n15, ExternalForce="0,0,0") {
    angularDampingOverride=0.9999
    angularSpringConstant=12000
    bAngularSpring=true
    bGravityOverrideInSimSpace=true
    bOverrideAngularDamping=true
    bUseGravityOverride=true
    gravityScale=0.35
    simulationSpace=RootRelative
    alphaBoolBlend {blendOption=Linear}
    boundBone {boneName=dyn_tassel_legs_06}
    componentAppliedLinearAccClamp=<3000 3000 3000>
    componentLinearAccScale=<0.4 0.4 0.4>
    componentLinearVelScale=<0.4 0.4 0.4>
    gravityOverride {z=-980}
    physicsBodyDefinitions [{
      sphereCollisionRadius=0
      boundBone {boneName=dyn_tassel_legs_06}
      boxExtents=<15 27 16>
      constraintSetup {angularLimitsMax=<15 15 15>, angularLimitsMin=<-15 -15 0>, angularTarget=<1 0 0>}
      localJointOffset=<4.5 0 0>
     }]
   }
   n40 = Anim Dynamics - Bone: dyn_tassel_legs_07(ComponentPose=@n39, Alpha=@n15, ExternalForce="0,0,0") {
    angularDampingOverride=0.9999
    angularSpringConstant=12000
    bAngularSpring=true
    bGravityOverrideInSimSpace=true
    bOverrideAngularDamping=true
    bUseGravityOverride=true
    gravityScale=0.35
    simulationSpace=RootRelative
    alphaBoolBlend {blendOption=Linear}
    boundBone {boneName=dyn_tassel_legs_07}
    componentAppliedLinearAccClamp=<3000 3000 3000>
    componentLinearAccScale=<0.4 0.4 0.4>
    componentLinearVelScale=<0.4 0.4 0.4>
    gravityOverride {z=-980}
    physicsBodyDefinitions [{
      sphereCollisionRadius=0
      boundBone {boneName=dyn_tassel_legs_07}
      boxExtents=<15 18 26>
      constraintSetup {angularLimitsMax=<15 15 15>, angularLimitsMin=<-15 -15 0>, angularTarget=<1 0 0>}
      localJointOffset=<4.5 0 0>
     }]
   }
   n41 = Anim Dynamics - Bone: dyn_tassel_legs_08(ComponentPose=@n40, Alpha=@n15, ExternalForce="0,0,0") {
    angularDampingOverride=0.9999
    angularSpringConstant=12000
    bAngularSpring=true
    bGravityOverrideInSimSpace=true
    bOverrideAngularDamping=true
    bUseGravityOverride=true
    gravityScale=0.35
    simulationSpace=RootRelative
    alphaBoolBlend {blendOption=Linear}
    boundBone {boneName=dyn_tassel_legs_08}
    componentAppliedLinearAccClamp=<3000 3000 3000>
    componentLinearAccScale=<0.4 0.4 0.4>
    componentLinearVelScale=<0.4 0.4 0.4>
    gravityOverride {z=-980}
    physicsBodyDefinitions [{
      sphereCollisionRadius=0
      boundBone {boneName=dyn_tassel_legs_08}
      boxExtents=<15 24 21>
      constraintSetup {angularLimitsMax=<15 15 15>, angularLimitsMin=<-15 -15 0>, angularTarget=<1 0 0>}
      localJointOffset=<4.5 0 0>
     }]
   }
   n42 = Anim Dynamics - Bone: dyn_tassel_legs_09(ComponentPose=@n41, Alpha=@n15, ExternalForce="0,0,0") {
    angularDampingOverride=0.9999
    angularSpringConstant=12000
    bAngularSpring=true
    bGravityOverrideInSimSpace=true
    bOverrideAngularDamping=true
    bUseGravityOverride=true
    gravityScale=0.35
    simulationSpace=RootRelative
    alphaBoolBlend {blendOption=Linear}
    boundBone {boneName=dyn_tassel_legs_09}
    componentAppliedLinearAccClamp=<3000 3000 3000>
    componentLinearAccScale=<0.4 0.4 0.4>
    componentLinearVelScale=<0.4 0.4 0.4>
    gravityOverride {z=-980}
    physicsBodyDefinitions [{
      sphereCollisionRadius=0
      boundBone {boneName=dyn_tassel_legs_09}
      boxExtents=<15 22 28>
      constraintSetup {angularLimitsMax=<15 15 15>, angularLimitsMin=<-15 -15 0>, angularTarget=<1 0 0>}
      localJointOffset=<4.5 0 0>
     }]
   }
   n43 = Anim Dynamics - Bone: dyn_tassel_legs_010(ComponentPose=@n42, Alpha=@n15, ExternalForce="0,0,0") {
    angularDampingOverride=0.9999
    angularSpringConstant=12000
    bAngularSpring=true
    bGravityOverrideInSimSpace=true
    bOverrideAngularDamping=true
    bUseGravityOverride=true
    gravityScale=0.35
    simulationSpace=RootRelative
    alphaBoolBlend {blendOption=Linear}
    boundBone {boneName=dyn_tassel_legs_010}
    componentAppliedLinearAccClamp=<3000 3000 3000>
    componentLinearAccScale=<0.4 0.4 0.4>
    componentLinearVelScale=<0.4 0.4 0.4>
    gravityOverride {z=-980}
    physicsBodyDefinitions [{
      sphereCollisionRadius=0
      boundBone {boneName=dyn_tassel_legs_010}
      boxExtents=<15 18 26>
      constraintSetup {angularLimitsMax=<15 15 15>, angularLimitsMin=<-15 -15 -15>, angularTarget=<1 0 0>}
      localJointOffset=<4.5 0 0>
     }]
   }
   n44 = Anim Dynamics - Bone: dyn_tassel_legs_011(ComponentPose=@n43, Alpha=@n15, ExternalForce="0,0,0") {
    angularDampingOverride=0.9999
    angularSpringConstant=12000
    bAngularSpring=true
    bGravityOverrideInSimSpace=true
    bOverrideAngularDamping=true
    bUseGravityOverride=true
    gravityScale=0.35
    simulationSpace=RootRelative
    alphaBoolBlend {blendOption=Linear}
    boundBone {boneName=dyn_tassel_legs_011}
    componentAppliedLinearAccClamp=<3000 3000 3000>
    componentLinearAccScale=<0.4 0.4 0.4>
    componentLinearVelScale=<0.4 0.4 0.4>
    gravityOverride {z=-980}
    physicsBodyDefinitions [{
      sphereCollisionRadius=0
      boundBone {boneName=dyn_tassel_legs_011}
      boxExtents=<15 22 28>
      constraintSetup {angularLimitsMax=<15 15 15>, angularLimitsMin=<-15 -15 -15>, angularTarget=<1 0 0>}
      localJointOffset=<4.5 0 0>
     }]
   }
   n45 = Anim Dynamics - Bone: dyn_tassel_legs_012(ComponentPose=@n44, Alpha=@n15, ExternalForce="0,0,0") {
    angularDampingOverride=0.9999
    angularSpringConstant=12000
    bAngularSpring=true
    bGravityOverrideInSimSpace=true
    bOverrideAngularDamping=true
    bUseGravityOverride=true
    gravityScale=0.35
    simulationSpace=RootRelative
    alphaBoolBlend {blendOption=Linear}
    boundBone {boneName=dyn_tassel_legs_012}
    componentAppliedLinearAccClamp=<3000 3000 3000>
    componentLinearAccScale=<0.4 0.4 0.4>
    componentLinearVelScale=<0.4 0.4 0.4>
    gravityOverride {z=-980}
    physicsBodyDefinitions [{
      sphereCollisionRadius=0
      boundBone {boneName=dyn_tassel_legs_012}
      boxExtents=<15 18 26>
      constraintSetup {angularLimitsMax=<15 15 15>, angularLimitsMin=<-15 -15 -15>, angularTarget=<1 0 0>}
      localJointOffset=<4.5 0 0>
     }]
   }
   n46 = Anim Dynamics - Bone: dyn_tassel_legs_013(ComponentPose=@n45, Alpha=@n15, ExternalForce="0,0,0") {
    angularDampingOverride=0.9999
    angularSpringConstant=12000
    bAngularSpring=true
    bGravityOverrideInSimSpace=true
    bOverrideAngularDamping=true
    bUseGravityOverride=true
    gravityScale=0.35
    simulationSpace=RootRelative
    alphaBoolBlend {blendOption=Linear}
    boundBone {boneName=dyn_tassel_legs_013}
    componentAppliedLinearAccClamp=<3000 3000 3000>
    componentLinearAccScale=<0.4 0.4 0.4>
    componentLinearVelScale=<0.4 0.4 0.4>
    gravityOverride {z=-980}
    physicsBodyDefinitions [{
      sphereCollisionRadius=0
      boundBone {boneName=dyn_tassel_legs_013}
      boxExtents=<15 18 26>
      constraintSetup {angularLimitsMax=<15 15 15>, angularLimitsMin=<-15 -15 -15>, angularTarget=<1 0 0>}
      localJointOffset=<4.5 0 0>
     }]
   }
   n47 = Anim Dynamics - Bone: dyn_tassel_legs_014(ComponentPose=@n46, Alpha=1, ExternalForce="0,0,0") {
    angularDampingOverride=0.9999
    angularSpringConstant=12000
    bAngularSpring=true
    bGravityOverrideInSimSpace=true
    bOverrideAngularDamping=true
    bUseGravityOverride=true
    gravityScale=0.35
    simulationSpace=RootRelative
    alphaBoolBlend {blendOption=Linear}
    boundBone {boneName=dyn_tassel_legs_014}
    componentAppliedLinearAccClamp=<3000 3000 3000>
    componentLinearAccScale=<0.4 0.4 0.4>
    componentLinearVelScale=<0.4 0.4 0.4>
    gravityOverride {z=-980}
    physicsBodyDefinitions [{
      sphereCollisionRadius=0
      boundBone {boneName=dyn_tassel_legs_014}
      boxExtents=<15 27 16>
      constraintSetup {angularLimitsMax=<15 15 15>, angularLimitsMin=<-15 -15 -15>, angularTarget=<1 0 0>}
      localJointOffset=<4.5 0 0>
     }]
   }
   n48 = Anim Dynamics - Bone: dyn_tassel_legs_015(ComponentPose=@n47, Alpha=@n15, ExternalForce="0,0,0") {
    angularDampingOverride=0.9999
    angularSpringConstant=12000
    bAngularSpring=true
    bGravityOverrideInSimSpace=true
    bOverrideAngularDamping=true
    bUseGravityOverride=true
    gravityScale=0.35
    simulationSpace=RootRelative
    alphaBoolBlend {blendOption=Linear}
    boundBone {boneName=dyn_tassel_legs_015}
    componentAppliedLinearAccClamp=<3000 3000 3000>
    componentLinearAccScale=<0.4 0.4 0.4>
    componentLinearVelScale=<0.4 0.4 0.4>
    gravityOverride {z=-980}
    physicsBodyDefinitions [{
      sphereCollisionRadius=0
      boundBone {boneName=dyn_tassel_legs_015}
      boxExtents=<15 24 21>
      constraintSetup {angularLimitsMax=<15 15 15>, angularLimitsMin=<-15 -15 0>, angularTarget=<1 0 0>}
      localJointOffset=<4.5 0 0>
     }]
   }
   n49 = Anim Dynamics - Bone: dyn_tassel_legs_016(ComponentPose=@n48, Alpha=@n15, ExternalForce="0,0,0") {
    angularDampingOverride=0.9999
    angularSpringConstant=12000
    bAngularSpring=true
    bGravityOverrideInSimSpace=true
    bOverrideAngularDamping=true
    bUseGravityOverride=true
    gravityScale=0.35
    simulationSpace=RootRelative
    alphaBoolBlend {blendOption=Linear}
    boundBone {boneName=dyn_tassel_legs_016}
    componentAppliedLinearAccClamp=<3000 3000 3000>
    componentLinearAccScale=<0.4 0.4 0.4>
    componentLinearVelScale=<0.4 0.4 0.4>
    gravityOverride {z=-980}
    physicsBodyDefinitions [{
      sphereCollisionRadius=0
      boundBone {boneName=dyn_tassel_legs_016}
      boxExtents=<15 18 26>
      constraintSetup {angularLimitsMax=<15 15 15>, angularLimitsMin=<-15 -15 0>, angularTarget=<1 0 0>}
      localJointOffset=<4.5 0 0>
     }]
   }
   n50 = Anim Dynamics - Bone: dyn_tassel_legs_017(ComponentPose=@n49, Alpha=@n15, ExternalForce="0,0,0") {
    angularDampingOverride=0.9999
    angularSpringConstant=12000
    bAngularSpring=true
    bGravityOverrideInSimSpace=true
    bOverrideAngularDamping=true
    bUseGravityOverride=true
    gravityScale=0.35
    simulationSpace=RootRelative
    alphaBoolBlend {blendOption=Linear}
    boundBone {boneName=dyn_tassel_legs_017}
    componentAppliedLinearAccClamp=<3000 3000 3000>
    componentLinearAccScale=<0.4 0.4 0.4>
    componentLinearVelScale=<0.4 0.4 0.4>
    gravityOverride {z=-980}
    physicsBodyDefinitions [{
      sphereCollisionRadius=0
      boundBone {boneName=dyn_tassel_legs_017}
      boxExtents=<15 18 26>
      constraintSetup {angularLimitsMax=<15 15 15>, angularLimitsMin=<-15 -15 0>, angularTarget=<1 0 0>}
      localJointOffset=<4.5 0 0>
     }]
   }
   n51 = Anim Dynamics - Bone: dyn_tassel_legs_018(ComponentPose=@n50, Alpha=@n15, ExternalForce="0,0,0") {
    angularDampingOverride=0.9999
    angularSpringConstant=12000
    bAngularSpring=true
    bGravityOverrideInSimSpace=true
    bOverrideAngularDamping=true
    bUseGravityOverride=true
    gravityScale=0.35
    simulationSpace=RootRelative
    alphaBoolBlend {blendOption=Linear}
    boundBone {boneName=dyn_tassel_legs_018}
    componentAppliedLinearAccClamp=<3000 3000 3000>
    componentLinearAccScale=<0.4 0.4 0.4>
    componentLinearVelScale=<0.4 0.4 0.4>
    gravityOverride {z=-980}
    physicsBodyDefinitions [{
      sphereCollisionRadius=0
      boundBone {boneName=dyn_tassel_legs_018}
      boxExtents=<15 27 16>
      constraintSetup {angularLimitsMax=<15 15 15>, angularLimitsMin=<-15 -15 0>, angularTarget=<1 0 0>}
      localJointOffset=<4.5 0 0>
     }]
   }
   n52 = Get LegStrapsAnimDynAlpha() -> (LegStrapsAnimDynAlpha:double)
   n53 = "Constraint - dyn_leg_straps_l (PP)"!(ComponentPose=@n51, Alpha=@n52, ConstraintWeights_0=0.9275, ConstraintWeights_1:float=0.0725) -> (Pose) {
    alphaBoolBlend {blendOption=Linear}
    boneToModify {boneName=dyn_leg_straps_l}
    constraintSetup [{targetBone {boneName=thigh_l}}, {targetBone {boneName=thigh_twist_01_l}}]
    constraintWeights [1,1]
   }
   n54 = Anim Dynamics - Bone: dyn_leg_straps_l(ComponentPose=@n53, Alpha=@n52, ExternalForce="0,0,0") {
    angularDampingOverride=0.99
    angularSpringConstant=3000
    bAngularSpring=true
    bGravityOverrideInSimSpace=true
    bOverrideAngularDamping=true
    bUseGravityOverride=true
    gravityScale=0.35
    simulationSpace=RootRelative
    alphaBoolBlend {blendOption=Linear}
    boundBone {boneName=dyn_leg_straps_l}
    componentAppliedLinearAccClamp=<3000 3000 3000>
    componentLinearAccScale=<0.4 0.4 0.4>
    componentLinearVelScale=<0.4 0.4 0.4>
    gravityOverride {z=-980}
    physicsBodyDefinitions [{
      sphereCollisionRadius=0
      boundBone {boneName=dyn_leg_straps_l}
      boxExtents=<15 27 16>
      constraintSetup {angularLimitsMax=<20 5 20>, angularLimitsMin=<-20 -20 -20>, angularTarget=<1 0 0>}
      localJointOffset=<-4.5 0 0>
     }]
   }
   n55 = Get ArmTasselsAnimDynAlpha() -> (ArmTasselsAnimDynAlpha:double)
   n56 = "Constraint - dyn_tassle_forearm_l (PPPP)"!(
    ComponentPose=@n54
    Alpha=@n55
    ConstraintWeights_0=0.897
    ConstraintWeights_1:float=0.0641
    ConstraintWeights_2:float=0.0246
    ConstraintWeights_3:float=0.0143
   ) -> (Pose) {
    alphaBoolBlend {blendOption=Linear}
    boneToModify {boneName=dyn_tassle_forearm_l}
    constraintSetup [
     {targetBone {boneName=lowerarm_twist_01_l}}
     {targetBone {boneName=hand_l}}
     {targetBone {boneName=lowerarm_l}}
     {targetBone {boneName=lowerarm_twist_02_l}}
    ]
    constraintWeights [1,1,1,1]
   }
   n57 = "Constraint - dyn_tassle_forearm_r (PPPP)"!(
    ComponentPose=@n56
    Alpha=@n55
    ConstraintWeights_0=0.897
    ConstraintWeights_1:float=0.0641
    ConstraintWeights_2:float=0.0246
    ConstraintWeights_3:float=0.0143
   ) -> (Pose) {
    alphaBoolBlend {blendOption=Linear}
    boneToModify {boneName=dyn_tassle_forearm_r}
    constraintSetup [
     {targetBone {boneName=lowerarm_twist_01_r}}
     {targetBone {boneName=hand_r}}
     {targetBone {boneName=lowerarm_r}}
     {targetBone {boneName=lowerarm_twist_02_r}}
    ]
    constraintWeights [1,1,1,1]
   }
   n58 = Anim Dynamics - Bone: dyn_tassle_forearm_l(ComponentPose=@n57, Alpha=@n55, ExternalForce="0,0,0") {
    angularDampingOverride=0.99
    angularSpringConstant=3000
    bAngularSpring=true
    bGravityOverrideInSimSpace=true
    bOverrideAngularDamping=true
    bUseGravityOverride=true
    gravityScale=0.35
    simulationSpace=RootRelative
    alphaBoolBlend {blendOption=Linear}
    boundBone {boneName=dyn_tassle_forearm_l}
    componentAppliedLinearAccClamp=<3000 3000 3000>
    componentLinearAccScale=<0.4 0.4 0.4>
    componentLinearVelScale=<0.4 0.4 0.4>
    gravityOverride {z=-980}
    physicsBodyDefinitions [{
      sphereCollisionRadius=0
      boundBone {boneName=dyn_tassle_forearm_l}
      boxExtents=<25 15 19>
      constraintSetup {angularTargetAxis=AxisY, twistAxis=AxisY, angularLimitsMax=<89 10 10>, angularLimitsMin=<-89 -10 -10>, angularTarget=<0 1 0>}
      localJointOffset=<0 -4.5 0>
     }]
   }
   n59 = Anim Dynamics - Bone: dyn_tassle_forearm_r(ComponentPose=@n58, Alpha=@n55, ExternalForce="0,0,0") {
    angularDampingOverride=0.99
    angularSpringConstant=3000
    bAngularSpring=true
    bGravityOverrideInSimSpace=true
    bOverrideAngularDamping=true
    bUseGravityOverride=true
    gravityScale=0.35
    simulationSpace=RootRelative
    alphaBoolBlend {blendOption=Linear}
    boundBone {boneName=dyn_tassle_forearm_r}
    componentAppliedLinearAccClamp=<3000 3000 3000>
    componentLinearAccScale=<0.4 0.4 0.4>
    componentLinearVelScale=<0.4 0.4 0.4>
    gravityOverride {z=-980}
    physicsBodyDefinitions [{
      sphereCollisionRadius=0
      boundBone {boneName=dyn_tassle_forearm_r}
      boxExtents=<28 13 21>
      constraintSetup {angularTargetAxis=AxisY, twistAxis=AxisY, angularLimitsMax=<89 10 10>, angularLimitsMin=<-89 -10 -10>, angularTarget=<0 1 0>}
      localJointOffset=<0 4.5 0>
     }]
   }
   n60 = Get in vehicle scale alpha() -> (in vehicle scale alpha:double)
   n61 = "Transform (Modify) Bone - Bone: dyn_skirt_fr_01_l"(ComponentPose=@n59, Alpha=@n60, Translation="0,0,0", Rotation="0,0,0", Scale="0.1,0.1,0.1") {scaleMode=BMM_Replace, alphaBoolBlend {blendOption=Linear}, boneToModify {boneName=dyn_skirt_fr_01_l}, scale=<0.1 0.1 0.1>}
   n62 = "Transform (Modify) Bone - Bone: dyn_skirt_frdiag_01_l"(ComponentPose=@n61, Alpha=@n60, Translation="0,0,0", Rotation="0,0,0", Scale="0.1,0.1,0.1") {scaleMode=BMM_Replace, alphaBoolBlend {blendOption=Linear}, boneToModify {boneName=dyn_skirt_frdiag_01_l}, scale=<0.1 0.1 0.1>}
   n63 = "Transform (Modify) Bone - Bone: dyn_skirt_out_01_l"(ComponentPose=@n62, Alpha=@n60, Translation="0,0,0", Rotation="0,0,0", Scale="0.1,0.1,0.1") {scaleMode=BMM_Replace, alphaBoolBlend {blendOption=Linear}, boneToModify {boneName=dyn_skirt_out_01_l}, scale=<0.1 0.1 0.1>}
   n64 = "Transform (Modify) Bone - Bone: dyn_skirt_bk_01_l"(ComponentPose=@n63, Alpha=@n60, Translation="0,0,0", Rotation="0,0,0", Scale="0.1,0.1,0.1") {scaleMode=BMM_Replace, alphaBoolBlend {blendOption=Linear}, boneToModify {boneName=dyn_skirt_bk_01_l}, scale=<0.1 0.1 0.1>}
   n65 = "Transform (Modify) Bone - Bone: dyn_skirt_mid_01"(ComponentPose=@n64, Alpha=@n60, Translation="0,0,0", Rotation="0,0,0", Scale="0.1,0.1,0.1") {scaleMode=BMM_Replace, alphaBoolBlend {blendOption=Linear}, boneToModify {boneName=dyn_skirt_mid_01}, scale=<0.1 0.1 0.1>}
   n66 = "Transform (Modify) Bone - Bone: dyn_skirt_fr_01_r"(ComponentPose=@n65, Alpha=@n60, Translation="0,0,0", Rotation="0,0,0", Scale="0.1,0.1,0.1") {scaleMode=BMM_Replace, alphaBoolBlend {blendOption=Linear}, boneToModify {boneName=dyn_skirt_fr_01_r}, scale=<0.1 0.1 0.1>}
   n67 = "Transform (Modify) Bone - Bone: dyn_skirt_frdiag_01_r"(ComponentPose=@n66, Alpha=@n60, Translation="0,0,0", Rotation="0,0,0", Scale="0.1,0.1,0.1") {scaleMode=BMM_Replace, alphaBoolBlend {blendOption=Linear}, boneToModify {boneName=dyn_skirt_frdiag_01_r}, scale=<0.1 0.1 0.1>}
   n68 = "Transform (Modify) Bone - Bone: dyn_skirt_out_01_r"(ComponentPose=@n67, Alpha=@n60, Translation="0,0,0", Rotation="0,0,0", Scale="0.1,0.1,0.1") {scaleMode=BMM_Replace, alphaBoolBlend {blendOption=Linear}, boneToModify {boneName=dyn_skirt_out_01_r}, scale=<0.1 0.1 0.1>}
   n69 = "Transform (Modify) Bone - Bone: dyn_skirt_bk_01_r"(ComponentPose=@n68, Alpha=@n60, Translation="0,0,0", Rotation="0,0,0", Scale="0.1,0.1,0.1") {scaleMode=BMM_Replace, alphaBoolBlend {blendOption=Linear}, boneToModify {boneName=dyn_skirt_bk_01_r}, scale=<0.1 0.1 0.1>}
   Output Pose(Result=Component To Local(ComponentPose=@n69)) {layerGroup=None}
   n70 = "Constraint - dyn_skirt_fr_01_l (PPPPP)"!(
    ComponentPose=Local To Component(LocalPose=Copy Pose From Mesh() {bUseAttachedParent=true})
    Alpha=1
    ConstraintWeights_0=0.1724
    ConstraintWeights_1:float=0.3204
    ConstraintWeights_2:float=0.0751
    ConstraintWeights_3:float=0.1773
    ConstraintWeights_4:float=0.2548
   ) -> (Pose) {
    alphaBoolBlend {blendOption=Linear}
    boneToModify {boneName=dyn_skirt_fr_01_l}
    constraintSetup [
     {targetBone {boneName=spine_01}}
     {targetBone {boneName=spine_02}}
     {targetBone {boneName=spine_03}}
     {targetBone {boneName=spine_04}}
     {targetBone {boneName=pelvis}}
    ]
    constraintWeights [1,1,1,1,1]
   }
   n71 = "Constraint - dyn_skirt_frdiag_01_l (PPPPP)"!(
    ComponentPose=@n70
    Alpha=1
    ConstraintWeights_0=0.0433
    ConstraintWeights_1:float=0.0563
    ConstraintWeights_2:float=0
    ConstraintWeights_3:float=0.0118
    ConstraintWeights_4:float=0.8887
   ) -> (Pose) {
    alphaBoolBlend {blendOption=Linear}
    boneToModify {boneName=dyn_skirt_frdiag_01_l}
    constraintSetup [
     {targetBone {boneName=spine_01}}
     {targetBone {boneName=spine_02}}
     {targetBone {boneName=spine_03}}
     {targetBone {boneName=spine_04}}
     {targetBone {boneName=pelvis}}
    ]
    constraintWeights [1,1,1,1,1]
   }
   n72 = "Constraint - dyn_skirt_fr_01_l (R)"!(ComponentPose=@n71, Alpha=1) -> (Pose) {
    alphaBoolBlend {blendOption=Linear}
    boneToModify {boneName=dyn_skirt_fr_01_l}
    constraintSetup [{transformType=Rotation, targetBone {boneName=thigh_twist_01_l}}]
    constraintWeights [1]
   }
   n73 = "Constraint - dyn_skirt_frdiag_01_l (R)"!(ComponentPose=@n72, Alpha=0.2) -> (Pose) {
    alphaBoolBlend {blendOption=Linear}
    boneToModify {boneName=dyn_skirt_frdiag_01_l}
    constraintSetup [{transformType=Rotation, targetBone {boneName=thigh_twist_01_l}}]
    constraintWeights [1]
   }
   Save cached pose 'SavedPose21857'(Pose=Component To Local(ComponentPose=@n73)) {cacheName=SavedPose21857}
   n74 = "Constraint - dyn_skirt_fr_03_l (R)"!(ComponentPose=Local To Component(LocalPose=Use cached pose 'SavedPose21857'()), Alpha=1) -> (Pose) {
    alphaBoolBlend {blendOption=Linear}
    boneToModify {boneName=dyn_skirt_fr_03_l}
    constraintSetup [{transformType=Rotation, targetBone {boneName=calf_l}}]
    constraintWeights [1]
   }
   n75 = "Constraint - dyn_skirt_fr_04_l (R)"!(ComponentPose=@n74, Alpha=1) -> (Pose) {
    alphaBoolBlend {blendOption=Linear}
    boneToModify {boneName=dyn_skirt_fr_04_l}
    constraintSetup [{transformType=Rotation, targetBone {boneName=calf_l}}]
    constraintWeights [1]
   }
   n76 = Component To Local(ComponentPose=@n75)
   n77 = Use cached pose 'SavedPose21857'()
   n78 = NOT Boolean(A=Get bRigidBodyEnabled())
   n79 = Get CrouchingWithRigidBodyEnabled() -> (CrouchingWithRigidBodyEnabled:float)
   n80 = "dyn_skirt_frdiag_01_l.rotateY = thigh_l.rotateZ * -0.4 - Bone Driven Controller"!(
    ComponentPose:Component Space Pose Link Structure=Local To Component(LocalPose=Blend Poses by bool(BlendPose_0=@n76, BlendPose_1=@n77, bActiveValue=@n78) {blendType=Linear})
    Alpha:float=@n79
   ) -> (Pose:Component Space Pose Link Structure) {
    is=title AnimGraphNode_BoneDrivenController
    bAffectTargetRotationY=true
    multiplier=-0.4
    sourceComponent=RotationZ
    alphaBoolBlend {blendOption=Linear}
    sourceBone {boneName=thigh_l}
    targetBone {boneName=dyn_skirt_frdiag_01_l}
   }
   n81 = "dyn_skirt_frdiag_01_l.rotateX = thigh_l.rotateZ * -0.6 - Bone Driven Controller"!(ComponentPose:Component Space Pose Link Structure=@n80, Alpha:float=@n79) -> (Pose:Component Space Pose Link Structure) {
    is=title AnimGraphNode_BoneDrivenController
    bAffectTargetRotationX=true
    multiplier=-0.6
    sourceComponent=RotationZ
    alphaBoolBlend {blendOption=Linear}
    sourceBone {boneName=thigh_l}
    targetBone {boneName=dyn_skirt_frdiag_01_l}
   }
   n82 = "dyn_skirt_fr_01_l.rotateX = thigh_l.rotateZ * -0.8 - Bone Driven Controller"!(ComponentPose:Component Space Pose Link Structure=@n81, Alpha:float=@n79) -> (Pose:Component Space Pose Link Structure) {
    is=title AnimGraphNode_BoneDrivenController
    bAffectTargetRotationX=true
    multiplier=-0.8
    sourceComponent=RotationZ
    alphaBoolBlend {blendOption=Linear}
    sourceBone {boneName=thigh_l}
    targetBone {boneName=dyn_skirt_fr_01_l}
   }
   n83 = "dyn_skirt_fr_01_l.rotateY = thigh_l.rotateZ * -0.3 - Bone Driven Controller"!(ComponentPose:Component Space Pose Link Structure=@n82, Alpha:float=@n79) -> (Pose:Component Space Pose Link Structure) {
    is=title AnimGraphNode_BoneDrivenController
    bAffectTargetRotationY=true
    multiplier=-0.3
    sourceComponent=RotationZ
    alphaBoolBlend {blendOption=Linear}
    sourceBone {boneName=thigh_l}
    targetBone {boneName=dyn_skirt_fr_01_l}
   }
   n84 = "dyn_skirt_frdiag_01_r.rotateY = thigh_r.rotateZ * -0.4 - Bone Driven Controller"!(ComponentPose:Component Space Pose Link Structure=@n83, Alpha:float=@n79) -> (Pose:Component Space Pose Link Structure) {
    is=title AnimGraphNode_BoneDrivenController
    bAffectTargetRotationY=true
    multiplier=-0.4
    sourceComponent=RotationZ
    alphaBoolBlend {blendOption=Linear}
    sourceBone {boneName=thigh_r}
    targetBone {boneName=dyn_skirt_frdiag_01_r}
   }
   n85 = "Constraint - dyn_skirt_fr_01_r (PPPPP)"!(
    ComponentPose=@n84
    Alpha=1
    ConstraintWeights_0=0.1724
    ConstraintWeights_1:float=0.3204
    ConstraintWeights_2:float=0.0751
    ConstraintWeights_3:float=0.1773
    ConstraintWeights_4:float=0.2548
   ) -> (Pose) {
    alphaBoolBlend {blendOption=Linear}
    boneToModify {boneName=dyn_skirt_fr_01_r}
    constraintSetup [
     {targetBone {boneName=spine_01}}
     {targetBone {boneName=spine_02}}
     {targetBone {boneName=spine_03}}
     {targetBone {boneName=spine_04}}
     {targetBone {boneName=pelvis}}
    ]
    constraintWeights [1,1,1,1,1]
   }
   n86 = "Constraint - dyn_skirt_frdiag_01_r (PPPPP)"!(
    ComponentPose=@n85
    Alpha=1
    ConstraintWeights_0=0.0433
    ConstraintWeights_1:float=0.0563
    ConstraintWeights_2:float=0
    ConstraintWeights_3:float=0.0118
    ConstraintWeights_4:float=0.8887
   ) -> (Pose) {
    alphaBoolBlend {blendOption=Linear}
    boneToModify {boneName=dyn_skirt_frdiag_01_r}
    constraintSetup [
     {targetBone {boneName=spine_01}}
     {targetBone {boneName=spine_02}}
     {targetBone {boneName=spine_03}}
     {targetBone {boneName=spine_04}}
     {targetBone {boneName=pelvis}}
    ]
    constraintWeights [1,1,1,1,1]
   }
   n87 = "Constraint - dyn_skirt_fr_01_r (R)"!(ComponentPose=@n86, Alpha=1) -> (Pose) {
    alphaBoolBlend {blendOption=Linear}
    boneToModify {boneName=dyn_skirt_fr_01_r}
    constraintSetup [{transformType=Rotation, targetBone {boneName=thigh_twist_01_r}}]
    constraintWeights [1]
   }
   n88 = "Constraint - dyn_skirt_frdiag_01_r (R)"!(ComponentPose=@n87, Alpha=0.2) -> (Pose) {
    alphaBoolBlend {blendOption=Linear}
    boneToModify {boneName=dyn_skirt_frdiag_01_r}
    constraintSetup [{transformType=Rotation, targetBone {boneName=thigh_twist_01_r}}]
    constraintWeights [1]
   }
   Save cached pose 'SavedPose13205'(Pose=Component To Local(ComponentPose=@n88)) {cacheName=SavedPose13205}
   Sequence Player 'GrooveJams'() {sequence=&/GrooveJams, startPosition=13.7}
   Sequence Player 'RunningAround_V2'() {playRate=0, sequence=&/RunningAround_V2, startPosition=21.91}
  ]}
 EventGraph {nodes [
   Event BlueprintUpdateAnimation()
   > n1 = Sequence!(execute) -> (then_0, then_1, then_2:Exec)
   > n2 = Branch(Condition=Get bIsFerretVehicleDriver() -> (bIsFerretVehicleDriver:bool))
   Set in vehicle rigid alpha(execute=[@n2.then, @n3.then], in vehicle rigid alpha:double=0) -> (Output_Get:double)
   > Set in vehicle scale alpha(in vehicle scale alpha:double=1) -> (Output_Get:double)
   n3 = Branch(execute=@n2.else, Condition=Get bIsOctopusVehicleDriver() -> (bIsOctopusVehicleDriver:bool))
   Set in vehicle rigid alpha(execute=@n3.else, in vehicle rigid alpha=1)
   > Set in vehicle scale alpha(in vehicle scale alpha=0)
   n4 = Branch(execute=@n1.then_1, Condition=CharacterItemDefinitionHasAnyMetaTag(Tags=Make Array("[0]"=(TagName="Cosmetics.Variant.Property.Stage0"))))
   > Set ScarfRbanAlpha(ScarfRbanAlpha:double=0) -> (Output_Get:double)
   > Set CoatRbanAlpha(CoatRbanAlpha:double=0) -> (Output_Get:double)
   > Set LegTasselsAnimDynAlpha(LegTasselsAnimDynAlpha:double=0) -> (Output_Get:double)
   > Set LegStrapsAnimDynAlpha(LegStrapsAnimDynAlpha:double=1) -> (Output_Get:double)
   > Set ArmTasselsAnimDynAlpha(ArmTasselsAnimDynAlpha:double=0) -> (Output_Get:double)
   > Set SpursRbanAlpha(SpursRbanAlpha:double=0) -> (Output_Get:double)
   n5 = Branch(execute=@n4.else, Condition=CharacterItemDefinitionHasAnyMetaTag(Tags=Make Array("[0]"=(TagName="Cosmetics.Variant.Property.Stage1"))))
   > Set ScarfRbanAlpha(ScarfRbanAlpha=0)
   > Set CoatRbanAlpha(CoatRbanAlpha=0)
   > Set LegTasselsAnimDynAlpha(LegTasselsAnimDynAlpha=0)
   > Set LegStrapsAnimDynAlpha(LegStrapsAnimDynAlpha=0)
   > Set ArmTasselsAnimDynAlpha(ArmTasselsAnimDynAlpha=0)
   > Set SpursRbanAlpha(SpursRbanAlpha=0)
   n6 = Branch(execute=@n5.else, Condition=CharacterItemDefinitionHasAnyMetaTag(Tags=Make Array("[0]"=(TagName="Cosmetics.Variant.Property.Stage2"))))
   > Set ScarfRbanAlpha(ScarfRbanAlpha=0)
   > Set CoatRbanAlpha(CoatRbanAlpha=0)
   > Set LegTasselsAnimDynAlpha(LegTasselsAnimDynAlpha=1)
   > Set LegStrapsAnimDynAlpha(LegStrapsAnimDynAlpha=0)
   > Set ArmTasselsAnimDynAlpha(ArmTasselsAnimDynAlpha=1)
   > Set SpursRbanAlpha(SpursRbanAlpha=0)
   n7 = Branch(execute=@n6.else, Condition=CharacterItemDefinitionHasAnyMetaTag(Tags=Make Array("[0]"=(TagName="Cosmetics.Variant.Property.Stage3"))))
   > Set ScarfRbanAlpha(ScarfRbanAlpha=1)
   > Set CoatRbanAlpha(CoatRbanAlpha=0)
   > Set LegTasselsAnimDynAlpha(LegTasselsAnimDynAlpha=1)
   > Set LegStrapsAnimDynAlpha(LegStrapsAnimDynAlpha=0)
   > Set ArmTasselsAnimDynAlpha(ArmTasselsAnimDynAlpha=1)
   > Set SpursRbanAlpha(SpursRbanAlpha=1)
   n8 = Branch(execute=@n7.else, Condition=CharacterItemDefinitionHasAnyMetaTag(Tags=Make Array("[0]"=(TagName="Cosmetics.Variant.Property.Stage4"))))
   > Set ScarfRbanAlpha(ScarfRbanAlpha=1)
   > Set CoatRbanAlpha(CoatRbanAlpha=1)
   > Set LegTasselsAnimDynAlpha(LegTasselsAnimDynAlpha=0)
   > Set LegStrapsAnimDynAlpha(LegStrapsAnimDynAlpha=0)
   > Set ArmTasselsAnimDynAlpha(ArmTasselsAnimDynAlpha=1)
   > Set SpursRbanAlpha(SpursRbanAlpha=1)
   Branch(execute=@n8.else, Condition=CharacterItemDefinitionHasAnyMetaTag(Tags=Make Array("[0]"=(TagName="Cosmetics.Variant.Property.Stage5"))))
   > Set ScarfRbanAlpha(ScarfRbanAlpha=1)
   > Set CoatRbanAlpha(CoatRbanAlpha=1)
   > Set LegTasselsAnimDynAlpha(LegTasselsAnimDynAlpha=0)
   > Set LegStrapsAnimDynAlpha(LegStrapsAnimDynAlpha=0)
   > Set ArmTasselsAnimDynAlpha(ArmTasselsAnimDynAlpha=0)
   > Set SpursRbanAlpha(SpursRbanAlpha=1)
   Set ScarfRbanAlphaFINAL(
    execute=@n1.then_2
    ScarfRbanAlphaFINAL:double=float * float(A=Get ScarfRbanAlpha() -> (ScarfRbanAlpha:double), B=Get dj_emote_alpha() -> (dj_emote_alpha:double))
   ) -> (Output_Get:double)
   > Set CoatRbanAlphaFINAL(CoatRbanAlphaFINAL:double=float * float(A=Get CoatRbanAlpha() -> (CoatRbanAlpha:double), B=Get in vehicle rigid alpha() -> (in vehicle rigid alpha:double))) -> (Output_Get:double)
   n9 = Event OnNotifyPartsStarted!() -> (OutputDelegate, then, Context:Name=None) {is=title K2Node_Event}
   > Branch(Condition="Equal (Name)"!(A:Name=@n9.Context, B:Name=Emote_DJ_Drop) -> (ReturnValue:bool) {is=title K2Node_CallFunction})
   > Set dj_emote_alpha(dj_emote_alpha:double=0) -> (Output_Get:double)
   n10 = Event OnNotifyPartsStopped!() -> (OutputDelegate, then, Context:Name=None) {is=title K2Node_Event}
   > Branch(Condition="Equal (Name)"(A=@n10.Context, B=Emote_DJ_Drop))
   > Set dj_emote_alpha(dj_emote_alpha=1)
   n11 = Event OnEndSkydiving()
   ResetDynamics(execute=[@n16.then, @n11.then, @n12.then, @n15.then], InTeleportType=TeleportPhysics)
   > ResetDynamics(InTeleportType=ResetPhysics)
   n12 = Event OnBeginSkydiving()
   n13 = Event OnBeginTargeting() {is=title K2Node_Event}
   n14 = Get bIsCrouching()
   n15 = Branch(execute=[@n13.then, @n17.then], Condition=AND Boolean(A=Get bIsTargeting(), B=NOT Boolean(A=@n14)))
   Event OnParachuteOpened()
   > n16 = Delay(Duration=1.5)
   n17 = Event OnEndCrouching()
   Event OnCharacterCustomizationLoaded()
  ]}
]
variables [
 ArmTasselsAnimDynAlpha {type=number}
 CoatRbanAlpha {type=number}
 CoatRbanAlphaFINAL {type=number}
 CurrentStage {type=integer}
 dj_emote_alpha {type=number, default=1}
 in vehicle rigid alpha {type=number, default=1}
 in vehicle scale alpha {type=number}
 LegStrapsAnimDynAlpha {type=number}
 LegTasselsAnimDynAlpha {type=number}
 ScarfRbanAlpha {type=number}
 ScarfRbanAlphaFINAL {type=number}
 SpursRbanAlpha {type=number}
]
