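# Physics asset for the medium female Cowgirl player body that, per its name, supplies collision shapes for a cape.
# Capsules cover the pelvis, spine, clavicles, head and feet, with crossed pairs on spine_02 through spine_04.
# Large capsules on a few dynamic skirt and leg-strap bones add lower-body volume, while thigh, calf, arm and neck bodies hold no shapes.
# It has no constraints, so it defines collision volumes rather than a jointed ragdoll.
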PhysicsAsset /Game/Characters/Player/Female/Medium/Bodies/F_MED_Cowgirl/Meshes/F_MED_Cowgirl_CapeColliders
bodies {
 calf_l []
 calf_r []
 clavicle_l [{shapeType=Capsule, length=5.9464, radius=12.8853, center=<8.373 -2.1131 -2.7154>, rotation=<rot -77.5566 177.3915 7.7314>}]
 clavicle_r [{shapeType=Capsule, length=6.9464, radius=10.0457, center=<-8.3962 2.1109 2.7205>, rotation=<rot -77.5565 177.3915 -172.2685>}]
 dyn_leg_straps_l [{
   shapeType=Capsule
   length=7.25
   radius=6.75
   shapeName=dyn_leg_straps_l_capsule
   center=<-6.6756 0 -1.1668>
   rotation=<rot -90 -179.156 179.1555>
  }]
 dyn_skirt_frdiag_04_l [{shapeType=Capsule, length=11.75, radius=18.25, shapeName=dyn_skirt_frdiag_04_l_capsule, center=<-5.6204 -15.7443 16.743>}]
 dyn_skirt_frdiag_04_r [{shapeType=Capsule, length=13, radius=18.25, shapeName=dyn_skirt_frdiag_04_r_capsule, center=<4.819 11.5428 -14.4786>}]
 dyn_skirt_mid_05_r [{
   shapeType=Capsule
   length=14
   radius=27.75
   shapeName=dyn_skirt_mid_05_r_capsule
   center=<17.1708 3.3554 -7.5681>
   rotation=<rot -10.0005 0 3>
  }]
 foot_l [{shapeType=Capsule, length=12.5982, radius=6.5127, center=<-2.2072 6.0534 -0.3367>, rotation=<rot 0.0388 1.4701 -89.3982>}]
 foot_r [{shapeType=Capsule, length=12.5982, radius=6.5127, center=<2.2072 -6.0534 0.3367>, rotation=<rot 0.0388 1.4701 90.6018>}]
 head [{shapeType=Capsule, length=0, radius=18, center=<9.0697 0.0746 0>, rotation=<rot 0 -90 0>}]
 lowerarm_l []
 lowerarm_r []
 neck_02 []
 pelvis [{shapeType=Capsule, length=6.6528, radius=16.6595, center=<-4.151 -1.7993 -0.1981>}]
 spine_01 [{shapeType=Capsule, length=10.7671, radius=9.7354, center=<-0.9573 -1.0873 0>}]
 spine_02 [
  {shapeType=Capsule, length=5.2984, radius=12.4854, center=<1.179 -2.9451 0>}
  {shapeType=Capsule, length=20.6875, radius=9.3125, center=<-1.8317 2.606 0>, rotation=<rot 90 0 0>}
 ]
 spine_03 [
  {shapeType=Capsule, length=2.2984, radius=10.9854, center=<1.0565 -0.3854 0>}
  {shapeType=Capsule, length=20.7188, radius=9.2812, center=<-1.2326 2.6059 0>, rotation=<rot 90 0 0>}
 ]
 spine_04 [
  {shapeType=Capsule, length=9.5796, radius=15.1416, center=<6.6514 -1.2534 0>}
  {shapeType=Capsule, length=10.0312, radius=9.2812, center=<-1.2326 1.308 0>, rotation=<rot 90 0 0>}
 ]
 spine_05 [{shapeType=Capsule, length=10.9859, radius=10.0791, center=<5.6377 0.3227 0>}]
 thigh_l []
 thigh_r []
 upperarm_l []
 upperarm_r []
}
constraints []
