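# Physics asset for the F_MED_Cowgirl T456 female player body, named as the collision set for its cloth.
# Capsules cover the pelvis, lower two spine bones, left thigh and right calf, leaving the other leg and foot bodies without shapes.
# A huge root capsule sits offset beneath the origin, its surface reaching about ground level.
# Constraints chain legs and spine through pelvis to root, with free angular motion on spine, pelvis and right hip.
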PhysicsAsset /Game/Characters/Player/Female/Medium/Bodies/F_MED_Cowgirl/Meshes/F_MED_Cowgirl_T456_ClothCollision
bodies {
 calf_l []
 calf_r [{shapeType=Capsule, length=25.8314, radius=6.2077, center=<32.2832 8.5075 0.361>, rotation=<rot -84.4982 99.1169 -88.6976>}]
 foot_l []
 foot_r []
 pelvis [{shapeType=Capsule, length=7.1621, radius=11.3577, center=<-3.3777 -0.4455 0>}]
 root [{shapeType=Capsule, length=206.0387, radius=264.8878, center=<0 -0.0019 -258.768>, rotation=<rot 0 89.9999 89.9999>}]
 spine_01 [{shapeType=Capsule, length=5.1619, radius=6.9405, center=<0 0.6526 0>}]
 spine_02 [{shapeType=Capsule, length=2.0818, radius=7.0199, center=<0 2.1171 -0.0001>}]
 thigh_l [{shapeType=Capsule, length=13.13, radius=3.3133, center=<-28.6066 7.8352 -3.2129>, rotation=<rot 84.9996 20.0016 0.0001>}]
 thigh_r []
}
constraints [
 {bone1Name=thigh_l, bone2Name=pelvis, swing1LimitDegrees=45, swing2LimitDegrees=45, twistLimitDegrees=45}
 {bone1Name=calf_l, bone2Name=thigh_l, swing1LimitDegrees=45, swing2LimitDegrees=45, twistLimitDegrees=45}
 {bone1Name=foot_l, bone2Name=calf_l, swing1LimitDegrees=45, swing2LimitDegrees=45, twistLimitDegrees=45}
 {bone1Name=calf_r, bone2Name=thigh_r, swing1LimitDegrees=45, swing2LimitDegrees=45, twistLimitDegrees=45}
 {bone1Name=foot_r, bone2Name=calf_r, swing1LimitDegrees=45, swing2LimitDegrees=45, twistLimitDegrees=45}
 {
  bone1Name=thigh_r
  bone2Name=pelvis
  swing1LimitDegrees=45
  swing1Motion=Free
  swing2LimitDegrees=45
  swing2Motion=Free
  twistLimitDegrees=45
  twistMotion=Free
 }
 {
  bone1Name=spine_01
  bone2Name=pelvis
  swing1LimitDegrees=45
  swing1Motion=Free
  swing2LimitDegrees=45
  swing2Motion=Free
  twistLimitDegrees=45
  twistMotion=Free
 }
 {
  bone1Name=spine_02
  bone2Name=spine_01
  swing1LimitDegrees=45
  swing1Motion=Free
  swing2LimitDegrees=45
  swing2Motion=Free
  twistLimitDegrees=45
  twistMotion=Free
 }
 {
  bone1Name=pelvis
  bone2Name=root
  swing1LimitDegrees=45
  swing1Motion=Free
  swing2LimitDegrees=45
  swing2Motion=Free
  twistLimitDegrees=45
  twistMotion=Free
 }
]
